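# Material function that outputs a 0-1 fade for the Kelp Linen environment, based on how close the camera is to a fixed world-space zone center.
# A SphereMask tests camera position against a hardcoded center and radius, with Object_Offset shifting the radius per element and Density setting hardness.
# The mask is doubled and saturated, and the LocationFade_Off static switch replaces it with 1.
# A radial gradient call and several radius, position and density nodes are left unconnected.
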MaterialFunction /ClydeCosmetics/Environments/KelpLinen/MF_KelpLinen_LocationFade
graph {nodes [
  Constant3Vector() {constant=<rgba -31136 -105728 -1920 1>}
  Constant() {desc=Default Radius - Don't Touch unless fade distance for entire zone is changing, r=23054.5273}
  n1 = CameraPositionWS()
  n2 = Constant3Vector() {constant=<rgba -27852 -50384 -63335.4922 1>}
  n3 = ScalarParameter() {
   desc="Use this to make a element fade faster or slower than other objects when the camera moves thru the zone.  Tweak carefully.  "
   group=Kelp_Location
   parameterName=Object_Offset
  }
  n4 = Add(A=Constant() {desc=Default Radius - Don't Touch unless fade distance for entire zone is changing, r=131922.5}, B=@n3)
  n5 = Multiply(
   A=SphereMask(
    A=@n1
    B=@n2
    Radius=@n4
    Hardness=ScalarParameter() {defaultValue=0.85, desc=Default for old values0.5, group=Kelp_Location, parameterName=Density}
   )
   B=Constant() {r=2}
  )
  FunctionOutput(None=StaticSwitchParameter(True=Constant() {r=1}, False=Saturate(None=@n5)) {group=Kelp_Location, parameterName=LocationFade_Off, sortPriority=31}) {outputName=Output}
  MaterialFunctionCall(
   UVs=ComponentMask(None=@n1) {g=true, r=true}
   CenterPosition=ComponentMask(None=@n2) {g=true, r=true}
   Radius=@n4
   Density=Constant() {r=1.7342}
  ) {materialFunction=&RadialGradientExponential}
  ScalarParameter() {
   defaultValue=23054.5273
   desc=Default Radius - Don't Touch unless fade distance for entire zone is changing
   group=Kelp_Location
   parameterName=Radius
  }
  ScalarParameter() {defaultValue=2, group=Kelp_Location, parameterName=Density_Multi}
 ]}
