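# Translucent, Nanite-capable character material for KelpLinen environment figures, animated through world position offset from an animation texture.
# Time, with parameter or position-based random offsets, drives a subUV transform function outputting offset and normals, optionally adding spline-thickened ribbon LODs.
# Emissive combines tinted fresnel glow, optional base textures, head-masked rim light and time-of-day eye glow.
# Comparing custom depth with pixel depth fakes backface culling for opacity.
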Material /ClydeCosmetics/Environments/KelpLinen/M_characterWPO_KelpLinen
graph {
 properties {
  allowTranslucentCustomDepthWrites=true
  bHasPixelAnimation=true
  bTangentSpaceNormal=false
  bUseMaterialAttributes=true
  bUsedWithInstancedStaticMeshes=true
  bUsedWithNanite=true
  bUsedWithNiagaraMeshParticles=true
  blendMode=BLEND_Translucent
  opacityMaskClipValue=0
  pixelDepthOffsetMode=PDOM_Legacy
  usedShadingModels=Default Lit
 }
 nodes [
  n1 = Time()
  n2 = Add(A=@n1, B=ScalarParameter() {parameterName=TimeOffset})
  Add(A=@n2, B=PerInstanceRandom())
  Constant() {r=0.333}
  n3 = ObjectPositionWS()
  n4 = MaterialFunctionCall() {materialFunction=&ObjectScale}
  n5 = Subtract(
   A=Subtract(A=WorldPosition().Z, B=ComponentMask(None=@n3) {b=true})
   B=Multiply(A=@n4.Scale X, B=ScalarParameter() {defaultValue=-70, parameterName=BottomFade_Offset})
  )
  n6 = Multiply(A=@n4.Scale X, B=ScalarParameter() {defaultValue=1166.9116, parameterName=BottomFade_Distance})
  NamedRerouteDeclaration(None=Saturate(None=Divide(A=@n5, B=@n6))) {name=BottomFade, nodeColor=<rgba 1 0 0.8672 1>}
  n7 = TextureSampleParameter2D() {parameterName=BaseColor, texture=&T_BlacksRGB}
  NamedRerouteDeclaration(None=@n7.RGB) {name=base color, nodeColor=<rgba 0.9922 0 1 1>}
  n8 = TextureObjectParameter() {
   parameterName=AnimationTexture
   samplerType=SAMPLERTYPE_LinearColor
   texture=/Game/ArtTools/characterWPO/animation/SoldierMannequin_Female_Emote_LetsBegin_CMM
  }
  n9 = MaterialFunctionCall(Float2=TextureProperty(Tex=@n8)) {materialFunction=&BreakOutFloat2Components}
  n10 = Divide(B=NamedRerouteDeclaration(None=@n9.R) {name=textureWidth, nodeColor=<rgba 0.0078 0 1 1>})
  NamedRerouteDeclaration(None=Multiply(A=NamedRerouteDeclaration(None=@n10) {name=texelWidth, nodeColor=<rgba 0 1 0.2266 1>}) {constB=0.5}) {name=halfTexelWidth, nodeColor=<rgba 0.0312 0 1 1>}
  NamedRerouteDeclaration(None=@n8) {name=AnimationTexture, nodeColor=<rgba 0.6172 0 1 1>}
  n11 = MaterialFunctionCall(Float3=ObjectPositionWS()) {materialFunction=&BreakOutFloat3Components}
  n12 = Add(A=@n11.R, B=@n11.R)
  n13 = Frac(None=Divide(A=Add(A=@n12, B=@n11.B)) {constB=10})
  n14 = ScalarParameter() {defaultValue=33.3, parameterName=RandomTimeOffset}
  n15 = StaticSwitchParameter(True=Add(A=Multiply(A=@n13, B=@n14), B=@n1), False=@n2) {parameterName=UseRandomOffsetPosition}
  n16 = Divide(B=NamedRerouteUsage()) {constA=30}
  n17 = Frac(None=Multiply(A=Multiply(A=@n15, B=@n16), B=ScalarParameter() {defaultValue=1, parameterName=PlaybackSpeed}))
  n18 = NamedRerouteUsage()
  NamedRerouteDeclaration(None=Add(A=Multiply(A=@n17, B=@n18), B=NamedRerouteUsage())) {name=u, nodeColor=<rgba 0 1 0.9766 1>}
  NamedRerouteDeclaration(None=Add(A=NamedRerouteUsage(), B=NamedRerouteUsage())) {name=uStart, nodeColor=<rgba 1 0.3516 0 1>}
  n19 = NamedRerouteUsage()
  NamedRerouteDeclaration(None=Subtract(A=Subtract(B=@n19), B=@n19)) {name=uWidth, nodeColor=<rgba 0.9922 1 0 1>}
  n20 = VertexInterpolator(VS=LocalPosition() {localOrigin=InstancePreSkinning}.XYZ)
  n21 = VectorParameter() {group=RimLight, parameterName=Head_Fade_Position, defaultValue=<rgba -5.8184 4.9922 168 1>}
  n22 = ScalarParameter() {defaultValue=22, group=RimLight, parameterName=Head_Fade_Radius}
  n23 = ScalarParameter() {defaultValue=1, group=RimLight, parameterName=head_contrast}
  n24 = OneMinus(None=NamedRerouteDeclaration(None=SphereMask(A=@n20.PS, B=@n21.RGB, Radius=@n22, Hardness=@n23)) {name=Face_Mask, nodeColor=<rgba 0.8281 1 0 1>})
  n25 = NamedRerouteUsage()
  n26 = Multiply(
   A=MaterialFunctionCall!(
    BaseColor=Constant3Vector() {constant=<rgba 1 1 1 0>}
    Gmap=ScalarParameter() {bUseCustomPrimitiveData=true, defaultValue=1, group=RimLight, parameterName=Rim_LightStr, primitiveDataIndex=1}
    LightVector=VectorParameter() {group=RimLight, parameterName=Light_Direction, defaultValue=<rgba 0 0 1 1>}.RGB
   ) {materialFunction=/Game/Packages/Fortress_SharedMaterials/Base_Material_Functions/MF_Rimlight_SingleLight}.Result
   B=Fresnel(ExponentIn=ScalarParameter() {defaultValue=5, group=RimLight, parameterName=ExponentIn}, Normal=Transform(None=@n25))
  )
  NamedRerouteDeclaration(None=Power(
    Base=Multiply(A=@n24, B=@n26)
    Exp=ScalarParameter() {bUseCustomPrimitiveData=true, defaultValue=2, group=RimLight, parameterName=RimPOw, primitiveDataIndex=10}
   )) {name=RimLight_mask, nodeColor=<rgba 0 1 0.3438 1>}
  n27 = ScalarParameter() {bUseCustomPrimitiveData=true, defaultValue=1, parameterName=Max_Opacity}
  n28 = NamedRerouteUsage()
  n29 = Constant() {r=8192}
  n30 = ComponentMask(None=SceneTexture() {sceneTextureId=PPI_CustomDepth}.Color) {r=true}
  n31 = Divide(A=@n29, B=PixelDepth().R)
  n32 = Constant() {r=0.0001}
  n33 = Ceil(None=Subtract(A=Divide(A=@n29, B=@n30), B=Add(A=@n31, B=@n32) {constB=0.0001}))
  NamedRerouteDeclaration(None=LinearInterpolate(A=Multiply(A=@n27, B=@n28), Alpha=Clamp(None=@n33)) {constA=0.5, constB=0}) {name=backface culling/Opacity, nodeColor=<rgba 1 0.0703 0 1>}
  NamedRerouteDeclaration(None=MaterialFunctionCall() {materialFunction=&/MF_KelpLinen_LocationFade}.Output) {name=Location Fade, nodeColor=<rgba 0 0.4141 1 1>}
  n34 = Multiply(A=NamedRerouteUsage(), B=ScalarParameter() {defaultValue=0.2, parameterName=BaseColor_Influence, primitiveDataIndex=11})
  n35 = ScalarParameter() {defaultValue=-1, parameterName=Fresnel_Glow}
  n36 = Multiply(A=OneMinus(None=NamedRerouteUsage()), B=Fresnel(ExponentIn=Constant() {r=1.8193}))
  n37 = VectorParameter() {bUseCustomPrimitiveData=true, parameterName=Base_Color, primitiveDataIndex=6, defaultValue=<rgba 0.2195 0.2195 0.2195 1>}
  n38 = Multiply(
   A=LinearInterpolate(
    B=LinearInterpolate(B=@n35, Alpha=@n36) {constA=1, constB=2}
    Alpha=CollectionParameter() {collection=/ClydeCosmetics/Environments/KelpLinen/Materials/MPC_KelpLinen, parameterName=Fresnel_Glow}
   ) {constA=1}
   B=@n37.RGB
  )
  n39 = VectorParameter() {bUseCustomPrimitiveData=true, group=RimLight, parameterName=Rim_Light_Color, primitiveDataIndex=2, defaultValue=<rgba 1 1 1 1>}
  n40 = NamedRerouteUsage()
  n41 = StaticSwitchParameter(
   True=StaticSwitchParameter(True=NamedRerouteUsage(), False=LinearInterpolate(A=@n38, B=@n39.RGB, Alpha=@n40) {constB=10}) {parameterName=Debug_Face_Fade}
   False=@n38
  ) {group=RimLight, parameterName=RimLight?}
  n42 = ScalarParameter() {defaultValue=1, parameterName=Min_Glow}
  n43 = ScalarParameter() {defaultValue=1.5, parameterName=Max_Glow}
  n44 = Fresnel(ExponentIn=ScalarParameter() {defaultValue=0.2, group=RimLight, parameterName=PlayerExp}, Normal=Transform(None=NamedRerouteUsage()))
  n45 = StaticSwitchParameter(True=Multiply(A=Add(A=@n34, B=@n41), B=LinearInterpolate(A=@n42, B=@n43, Alpha=@n44) {constA=1}), False=@n41) {parameterName=UseBaseTextures}
  n46 = ScalarParameter() {defaultValue=100, parameterName=Eye_Str}
  n47 = Constant() {r=0.5}
  n48 = Constant() {r=1}
  n49 = MaterialFunctionCall(Value_Morning=@n47, Value_Day=@n47, Value_Evening=@n48, Value_Night=@n48) {materialFunction=&Scalar_Time_of_Day}
  n50 = MaterialFunctionCall(Emissive=Multiply(A=Constant3Vector() {constant=<rgba 0 0.8935 1 1>}, B=Multiply(A=@n46, B=@n49.Result))) {materialFunction=&MF_FX_EmissiveAdjustments}
  n51 = TextureSample() {samplerType=SAMPLERTYPE_LinearColor, texture=/ClydeCosmetics/Environments/KelpLinen/VertexBakes/BakeData/T_KelpLinen_Emissive}
  NamedRerouteDeclaration(None=StaticSwitchParameter(True=LinearInterpolate(A=@n45, B=@n50.Result, Alpha=@n51.R), False=@n45) {parameterName=eye_Glow}) {name=emissive, nodeColor=<rgba 0 1 0.1094 1>}
  NamedRerouteDeclaration(None=Divide(B=NamedRerouteDeclaration(None=@n9.G) {name=textureHeight, nodeColor=<rgba 0.5469 0 1 1>})) {name=texelHeight, nodeColor=<rgba 0 1 0.2266 1>}
  n52 = TextureSampleParameter2D() {parameterName=Normals, samplerType=SAMPLERTYPE_Normal, texture=&FlatNormalBlue}
  NamedRerouteDeclaration(None=@n52.RGB) {name=normal, nodeColor=<rgba 0 1 0.1094 1>}
  NamedRerouteUsage()
  Saturate(None=MaterialFunctionCall(Contrast=Constant() {r=0.03}) {materialFunction=&CheapContrast}.Result)
  n53 = VertexColor()
  n54 = MaterialFunctionCall(Hue Shift Percentage=VertexColor().R, Texture=Constant3Vector() {constant=<rgba 1 0 0 0>}) {materialFunction=&HueShift}
  n55 = ScalarParameter() {defaultValue=20, parameterName=Width}
  n56 = MaterialFunctionCall(WidthBase=@n55, WidthTip=@n55) {materialFunction=&SplineThicken}
  n57 = MaterialFunctionCall!(u=NamedRerouteUsage(), bind_u=NamedRerouteUsage(), vStart=Constant(), vHeight=Constant() {r=1}, texelHeight=NamedRerouteUsage(), WorldNormal) {materialFunction=/Game/ArtTools/characterWPO/MF_WPOTransform_subuvs}
  n58 = SetMaterialAttributes!(
   MaterialAttributes
   Base Color=StaticSwitchParameter(
    True=StaticSwitchParameter(True=@n53, False=@n54.Result) {parameterName=PreviewRaw}
    False=StaticSwitchParameter(True=@n7.RGB, False=@n38) {parameterName=UseBaseTextures}
   ) {parameterName=PreviewWeights}
   Emissive Color=NamedRerouteUsage()
   Normal=StaticSwitchParameter(
    True=@n56.Normal
    False=MaterialFunctionCall(BaseNormal=@n57.Normal, AdditionalNormal=@n52.RGB) {materialFunction=&BlendAngleCorrectedNormals}.Result
   ) {parameterName=IsRibbonLOD}
   World Position Offset=StaticSwitchParameter(True=Add(A=@n57.WPO, B=@n56.WorldPositionOffset), False=@n57.WPO) {parameterName=IsRibbonLOD}
   Roughness=Constant() {r=1}
   Specular=Constant() {r=0.5}
  ) {attributeSetTypes [
    #1
    #2
    #3
    #4
    #5
    #6
   ]}
  n59 = NamedRerouteUsage()
  SetMaterialAttributes!(
   MaterialAttributes=SetMaterialAttributes!(MaterialAttributes=@n58, Opacity=NamedRerouteUsage()) {attributeSetTypes [#7]}
   Opacity Mask=LinearInterpolate(Alpha=MaterialFunctionCall(Alpha Threshold=@n59) {materialFunction=&DitherTemporalAA}.Result) {constA=-0.5}
  ) {attributeSetTypes [#8]}
  n60 = ScalarParameter() {parameterName=jointIndex}
  n61 = Divide(A=NamedRerouteUsage())
  StaticSwitchParameter(True=SphereMask(A=@n53.R, B=Divide(A=@n60, B=@n61)) {attenuationRadius=0.01}, False=@n54.Result) {parameterName=pickWeight}
  Subtract(
   A=TransformPosition(None=WorldPosition().XYZ) {transformSourceType=TRANSFORMPOSSOURCE_World}
   B=Transform(None=ObjectPositionWS()) {transformSourceType=TRANSFORMSOURCE_World, transformType=TRANSFORM_Local}
  )
 ]
}
