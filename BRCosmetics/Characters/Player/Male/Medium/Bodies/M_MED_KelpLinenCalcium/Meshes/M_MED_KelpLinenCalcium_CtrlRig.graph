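# Control Rig for the M_MED_KelpLinenCalcium character body, adding corrective secondary motion to its hip chain, jacket, T-shirt, vest sleeves, badges and zippers.
# After a shared DeformRig function, weighted parent constraints and spherical pose readers on thighs, upper arms and spine drive bone offsets.
# Crouch, fly, DBNO and front-end flags adjust those offsets.
# Spline IK shapes front and back chains, whose controls blend between construction-cached bind and raised-thigh poses.
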ControlRigBlueprint /BRCosmetics/Characters/Player/Male/Medium/Bodies/M_MED_KelpLinenCalcium/Meshes/M_MED_KelpLinenCalcium_CtrlRig
graphs [
 RigVMFunctionLibrary {links [], nodes []}
 RigVMFunction_Sequence_1_1_ContainedGraph {nodes [
   n1 = RigVMFunctionEntryNode()
   n2 = &RigVMFunction_Sequence(ExecuteContext=@n1.ExecuteContext)
   n3 = &RigVMFunction_Sequence(ExecuteContext=@n2.B)
   n4 = &RigVMFunction_Sequence(ExecuteContext=@n3.B)
   RigVMFunctionReturnNode(A=@n2.A, B=@n3.A, C=@n4.A, D=@n4.B)
  ]}
 RigVMFunction_Sequence_1_1_ContainedGraph {nodes [
   n1 = RigVMFunctionEntryNode()
   n2 = &RigVMFunction_Sequence(ExecuteContext=@n1.ExecuteContext)
   n3 = &RigVMFunction_Sequence(ExecuteContext=@n2.B)
   RigVMFunctionReturnNode(A=@n2.A, B=@n3.A, C=@n3.B)
  ]}
 RigVMFunction_Sequence_2_1_ContainedGraph {nodes [
   n1 = RigVMFunctionEntryNode()
   n2 = &RigVMFunction_Sequence(ExecuteContext=@n1.ExecuteContext)
   n3 = &RigVMFunction_Sequence(ExecuteContext=@n2.B)
   n4 = &RigVMFunction_Sequence(ExecuteContext=@n3.B)
   RigVMFunctionReturnNode(A=@n2.A, B=@n3.A, C=@n4.A, D=@n4.B)
  ]}
 RigVMFunction_Sequence_3_1_ContainedGraph {nodes [
   n1 = RigVMFunctionEntryNode()
   n2 = &RigVMFunction_Sequence(ExecuteContext=@n1.ExecuteContext)
   n3 = &RigVMFunction_Sequence(ExecuteContext=@n2.B)
   n4 = &RigVMFunction_Sequence(ExecuteContext=@n3.B)
   RigVMFunctionReturnNode(A=@n2.A, B=@n3.A, C=@n4.A, D=@n4.B)
  ]}
 RigVMModel {nodes [
   RigUnit_BeginExecution = &RigUnit_BeginExecution()
   FN_DeformRig_FNC = RigVMFunctionReferenceNode(
    DeformRig_ Curve="(deformRig_l_upperarm_up_alpha_72_#1=False,deformRig_r_upperarm_up_alpha_73_#2=False,deformRig_l_upperarm_fr_alpha_74_#3=False,deformRig_r_upperarm_fr_alpha_75_CB51B8 ...(truncated)"
    DeformRig_Bool="(upperarm_1_#4=True,elbow_3_#5=True,wrist_7_#6=True,groin_8_#7=True,glute_17_#8=True,knee_9_40C408324CA261 ...(truncated)"
    DeformRig_Debug=(upperarm_debug_18_#4=False,wrist_debug_20_#6=False,lat_debug_26_#9=False,groinGlute_debug_27_#7=False)
    DeformRig_Value="(upperarm_up_18_#10=(Rotation=(X=0,Y=0,Z=0,W=1),Translation=(X=0,Y=0,Z=8),Scale3D=(X=0.7,Y=1,Z=1)),upperarm_fwd_fr_19_#11=(Rotation=(X=0,Y=0,Z=0,W=1),Translation=(X=0,Y=10,Z=0),Scale3D=(X=0.5,Y=1,Z= ...(truncated)"
    ExecuteContext=@RigUnit_BeginExecution.ExecutePin
   )
   RigVMFunction_Sequence = &RigVMFunction_Sequence(ExecuteContext=@FN_DeformRig_FNC.ExecuteContext)
   &RigUnit_ModifyTransforms(
    ItemToModify=((Item=(Type=Bone,Name="dyn_chain_control_1"),Transform=(Rotation=(X=0,Y=0,Z=0,W=1),Translation=(X=-4,Y=2,Z=2),Scale3D=(X=1,Y=1,Z=1))))
    WeightMinimum=0
    Weight=&RigVMFunction_MathDoubleMul(A=1, B=RigVMVariableNode() {variable=spr_thigh_r_up_alpha}.Value).Result
    ItemToModify.0.Transform.Rotation=&RigVMFunction_MathQuaternionFromEuler(Euler=(X=-10,Y=5,Z=-15)).Result
    ExecutePin=@RigVMFunction_Sequence.A
   )
   > RigVMFunction_Sequence_2_1 = RigVMAggregateNode()
   &RigUnit_ParentConstraint(
    AdvancedSettings=(RotationOrderForFilter=XZY)
    Child=(Type=Bone,Name="dyn_zipper_base_l")
    Parents=((Item=(Type=Bone,Name="thigh_l"),Weight=0.295),(Item=(Type=Bone,Name="thigh_twist_01_l"),Weight=0.705))
    Weight=1
    ExecutePin=@RigVMFunction_Sequence_2_1.A
   )
   > &RigUnit_ParentConstraint(
    AdvancedSettings=(RotationOrderForFilter=XZY)
    Child=(Type=Bone,Name="dyn_zipper_base_r")
    Parents=((Item=(Type=Bone,Name="thigh_r"),Weight=0.295),(Item=(Type=Bone,Name="thigh_twist_01_r"),Weight=0.705))
    Weight=1
   )
   > &RigUnit_ParentConstraint(
    AdvancedSettings=(RotationOrderForFilter=XZY)
    Child=(Type=Bone,Name="dyn_badge_1")
    Parents="((Item=(Type=Bone,Name=\"deform_upperarm_up_l\"),Weight=0.0009),(Item=(Type=Bone,Name=\"upperarm_twist_01_l\"),Weight=0.0073),(Item=(Type=Bone,Name=\"spine_03\"),Weight=0.0442),(Item=(Type=Bone,Name=\"dyn_vest_sleeve_l\"),Weight=0.0569),(Item=(Type=Bone,Name=\"spin ...(truncated)"
    Weight=1
   )
   > &RigUnit_ParentConstraint(
    AdvancedSettings=(RotationOrderForFilter=XZY)
    Child=(Type=Bone,Name="dyn_badge_2")
    Parents="((Item=(Type=Bone,Name=\"dyn_jacket_2_2\"),Weight=0.0036),(Item=(Type=Bone,Name=\"dyn_vest_sleeve_l\"),Weight=0.0199),(Item=(Type=Bone,Name=\"deform_upperarm_fr_l\"),Weight=0.0488),(Item=(Type=Bone,Name=\"spine_03\"),Weight=0.0654),(Item=(Type=Bone,Name=\"spine_04\" ...(truncated)"
    Weight=1
   )
   &RigUnit_ParentConstraint(
    AdvancedSettings=(RotationOrderForFilter=XZY)
    Child=(Type=Null,Name="dyn_chain_front_Main_null")
    Parents=((Item=(Type=Bone,Name="dyn_chain_control_3"),Weight=0.25),(Item=(Type=Bone,Name="thigh_r"),Weight=0.75))
    Weight=1
    ExecutePin=@RigVMFunction_Sequence_2_1.B
   )
   > &RigUnit_ParentConstraint(
    AdvancedSettings=(RotationOrderForFilter=XZY)
    Child=(Type=Null,Name="dyn_chain_front_1_null")
    Parents=((Item=(Type=Bone,Name="pelvis"),Weight=0.62),(Item=(Type=Bone,Name="deform_groin_r"),Weight=0.19),(Item=(Type=Bone,Name="thigh_twist_01_r"),Weight=0.15),(Item=(Type=Bone,Name="spine_01"),Weight=0.04))
    Weight=1
   )
   > &RigUnit_ParentConstraint(
    AdvancedSettings=(RotationOrderForFilter=XZY)
    Child=(Type=Null,Name="dyn_chain_back_1_null")
    Parents=((Item=(Type=Bone,Name="pelvis"),Weight=0.77),(Item=(Type=Bone,Name="deform_glute_r"),Weight=0.17),(Item=(Type=Bone,Name="thigh_twist_01_r"),Weight=0.06))
    Weight=1
   )
   > n1 = &RigUnit_SphericalPoseReader(
    ActiveRegionScaleFactors=(PositiveWidth=2,NegativeWidth=0.5,PositiveHeight=2,NegativeHeight=2)
    ActiveRegionSize=0.1
    Debug=(bDrawDebug=true,DebugScale=25,DebugSegments=20,DebugThickness=0.25)
    DriverAxis=(X=1,Y=0,Z=0)
    DriverItem=(Type=Bone,Name="thigh_r")
    FalloffRegionScaleFactors=(PositiveWidth=1,NegativeWidth=1,PositiveHeight=1,NegativeHeight=1)
    FalloffSize=0.45
    FlipHeightScaling=false
    FlipWidthScaling=true
    OptionalParentItem=(Type=Bone)
    RotationOffset=(X=-90,Y=0,Z=0)
   )
   > RigVMVariableNode(Value=@n1.OutputParam) {variable=spr_thigh_r_up_alpha}
   > n2 = &RigUnit_SphericalPoseReader(
    ActiveRegionScaleFactors=(PositiveWidth=1,NegativeWidth=1,PositiveHeight=1,NegativeHeight=1)
    ActiveRegionSize=0.05
    Debug=(bDrawDebug=true,DebugScale=25,DebugSegments=20,DebugThickness=0.25)
    DriverAxis=(X=1,Y=0,Z=0)
    DriverItem=(Type=Bone,Name="thigh_r")
    FalloffRegionScaleFactors=(PositiveWidth=1,NegativeWidth=1,PositiveHeight=1,NegativeHeight=1)
    FalloffSize=0.09
    OptionalParentItem=(Type=Bone)
    RotationOffset=(X=15,Y=-80,Z=0)
   )
   > RigVMVariableNode(Value=@n2.OutputParam) {variable=spr_thigh_r_slight_back_out_apha}
   > RigVMFunction_Sequence_3_1 = RigVMAggregateNode()
   n3 = RigVMVariableNode() {variable=spr_thigh_r_up_alpha}
   n4 = &RigVMFunction_MathDoubleSub(A=1, B=@n3.Value)
   &RigUnit_ParentConstraint(
    AdvancedSettings=(RotationOrderForFilter=XZY)
    Child=(Type=Null,Name="dyn_chain_front_4_null")
    Parents=((Item=(Type=Bone,Name="thigh_r"),Weight=0.25),(Item=(Type=Control,Name="dyn_chain_front_Main_ctrl"),Weight=0.5),(Item=(Type=Bone,Name="deform_groin_r"),Weight=0.25))
    ExecutePin=@RigVMFunction_Sequence_3_1.A
    Weight=@n4.Result
   )
   > &RigUnit_ParentConstraint(
    AdvancedSettings=(RotationOrderForFilter=XZY)
    Child=(Type=Null,Name="dyn_chain_front_6_null")
    Parents=((Item=(Type=Bone,Name="thigh_r"),Weight=0.15),(Item=(Type=Control,Name="dyn_chain_front_Main_ctrl"),Weight=0.85))
    Weight=@n4.Result
   )
   > &RigUnit_ParentConstraint(
    AdvancedSettings=(RotationOrderForFilter=XZY)
    Child=(Type=Null,Name="dyn_chain_back_4_null")
    Parents=((Item=(Type=Bone,Name="thigh_r"),Weight=0.3),(Item=(Type=Control,Name="dyn_chain_front_Main_ctrl"),Weight=0.3),(Item=(Type=Control,Name="dyn_chain_back_1_ctrl"),Weight=0.4))
    Weight=@n4.Result
   )
   > &RigUnit_ParentConstraint(
    AdvancedSettings=(RotationOrderForFilter=XZY)
    Child=(Type=Null,Name="dyn_chain_back_7_null")
    Parents=((Item=(Type=Bone,Name="thigh_r"),Weight=0.25),(Item=(Type=Control,Name="dyn_chain_front_Main_ctrl"),Weight=0.5),(Item=(Type=Control,Name="dyn_chain_back_1_ctrl"),Weight=0.25))
    Weight=@n4.Result
   )
   > &RigUnit_ParentConstraint(
    AdvancedSettings=(RotationOrderForFilter=XZY)
    Child=(Type=Null,Name="dyn_chain_front_4_null")
    Parents=((Item=(Type=Bone,Name="thigh_r"),Weight=0.33),(Item=(Type=Control,Name="dyn_chain_front_Main_ctrl"),Weight=0.5),(Item=(Type=Bone,Name="deform_groin_r"),Weight=0.17))
    Weight=@n3.Value
   )
   > &RigUnit_ParentConstraint(
    AdvancedSettings=(RotationOrderForFilter=XZY)
    Child=(Type=Null,Name="dyn_chain_front_6_null")
    Parents=((Item=(Type=Bone,Name="thigh_r"),Weight=0.25),(Item=(Type=Control,Name="dyn_chain_front_Main_ctrl"),Weight=0.75))
    Weight=@n3.Value
   )
   > &RigUnit_ParentConstraint(
    AdvancedSettings=(RotationOrderForFilter=XZY)
    Child=(Type=Null,Name="dyn_chain_back_4_null")
    Parents=((Item=(Type=Control,Name="dyn_chain_back_1_ctrl"),Weight=0.5),(Item=(Type=Control,Name="dyn_chain_front_Main_ctrl"),Weight=0.5))
    Weight=@n3.Value
   )
   > &RigUnit_ParentConstraint(
    AdvancedSettings=(RotationOrderForFilter=XZY)
    Child=(Type=Null,Name="dyn_chain_back_7_null")
    Parents=((Item=(Type=Control,Name="dyn_chain_back_1_ctrl"),Weight=0.25),(Item=(Type=Control,Name="dyn_chain_front_Main_ctrl"),Weight=0.75))
    Weight=@n3.Value
   )
   Subtract = &RigVMFunction_MathDoubleSub(A=1, B=&RigVMFunction_MathBoolToFloat(Value=RigVMVariableNode() {variable=is_crouch}.Value).Result)
   n5 = RigVMVariableNode() {variable=spr_thigh_r_up_alpha}
   Multiply = &RigVMFunction_MathDoubleMul(B=@Subtract.Result, A=@n5.Value)
   > &RigUnit_ModifyTransforms(
    ItemToModify=((Item=(Type=Null,Name="dyn_chain_front_1_null"),Transform=(Rotation=(X=0,Y=0,Z=0,W=1),Translation=(X=0,Y=-1,Z=-1.5),Scale3D=(X=1,Y=1,Z=1))))
    WeightMinimum=0
    Weight=@Multiply.Result
   )
   > &RigUnit_ModifyTransforms(
    ItemToModify=((Item=(Type=Null,Name="dyn_chain_front_4_null"),Transform=(Rotation=(X=0,Y=0,Z=0,W=1),Translation=(X=0,Y=-2.5,Z=-3),Scale3D=(X=1,Y=1,Z=1))))
    WeightMinimum=0
    Weight=@Multiply.Result
   )
   > &RigUnit_ModifyTransforms(
    ItemToModify=((Item=(Type=Null,Name="dyn_chain_front_6_null"),Transform=(Rotation=(X=0,Y=0,Z=0,W=1),Translation=(X=-1,Y=2,Z=-1),Scale3D=(X=1,Y=1,Z=1))))
    WeightMinimum=0
    Weight=@Multiply.Result
   )
   > &RigUnit_ModifyTransforms(
    ItemToModify=((Item=(Type=Control,Name="dyn_chain_front_8_ctrl"),Transform=(Rotation=(X=0,Y=0,Z=0,W=1),Translation=(X=0,Y=3,Z=-1.5),Scale3D=(X=1,Y=1,Z=1))))
    WeightMinimum=0
    Weight=@Multiply.Result
   )
   > &RigUnit_ModifyTransforms(
    ItemToModify=((Item=(Type=Null,Name="dyn_chain_back_4_null"),Transform=(Rotation=(X=0,Y=0,Z=0,W=1),Translation=(X=0,Y=1,Z=-2),Scale3D=(X=1,Y=1,Z=1))))
    WeightMinimum=0
    Weight=@n5.Value
   )
   > &RigUnit_ModifyTransforms(
    ItemToModify=((Item=(Type=Null,Name="dyn_chain_back_7_null"),Transform=(Rotation=(X=0,Y=0,Z=0,W=1),Translation=(X=0,Y=-5,Z=0),Scale3D=(X=1,Y=1,Z=1))))
    WeightMinimum=0
    Weight=@Multiply.Result
   )
   > &RigUnit_ModifyTransforms(
    ItemToModify=((Item=(Type=Control,Name="dyn_chain_back_9_ctrl"),Transform=(Rotation=(X=0,Y=0,Z=0,W=1),Translation=(X=0,Y=2.5,Z=0),Scale3D=(X=1,Y=1,Z=1))))
    WeightMinimum=0
    Weight=@Multiply.Result
   )
   > &RigUnit_ModifyTransforms(
    ItemToModify=((Item=(Type=Null,Name="dyn_chain_front_Main_null"),Transform=(Rotation=(X=0,Y=0,Z=0,W=1),Translation=(X=0,Y=3,Z=-2),Scale3D=(X=1,Y=1,Z=1))))
    WeightMinimum=0
    Weight=@Multiply.Result
   )
   n6 = &RigUnit_SphericalPoseReader(
    ActiveRegionScaleFactors=(PositiveWidth=1,NegativeWidth=0.5,PositiveHeight=1,NegativeHeight=1)
    ActiveRegionSize=0.1
    Debug=(bDrawDebug=true,DebugScale=25,DebugSegments=20,DebugThickness=0.25)
    DriverAxis=(X=1,Y=0,Z=0)
    DriverItem=(Type=Bone,Name="thigh_r")
    FalloffRegionScaleFactors=(PositiveWidth=1,NegativeWidth=1,PositiveHeight=1,NegativeHeight=1)
    FalloffSize=0.35
    FlipHeightScaling=false
    FlipWidthScaling=true
    OptionalParentItem=(Type=Bone)
    RotationOffset=(X=-35,Y=-30,Z=20)
    ExecutePin=@RigVMFunction_Sequence_3_1.B
   )
   > &RigUnit_ModifyTransforms(
    ItemToModify=((Item=(Type=Control,Name="dyn_chain_front_1_ctrl"),Transform=(Rotation=(X=0,Y=0,Z=0,W=1),Translation=(X=0,Y=-2,Z=1),Scale3D=(X=1,Y=1,Z=1))))
    WeightMinimum=0
    Weight=@n6.OutputParam
   )
   > &RigUnit_ModifyTransforms(
    ItemToModify=((Item=(Type=Control,Name="dyn_chain_front_4_ctrl"),Transform=(Rotation=(X=0,Y=0,Z=0,W=1),Translation=(X=0,Y=-1,Z=-1),Scale3D=(X=1,Y=1,Z=1))))
    WeightMinimum=0
    Weight=@n6.OutputParam
   )
   > &RigUnit_ModifyTransforms(
    ItemToModify=((Item=(Type=Control,Name="dyn_chain_front_6_ctrl"),Transform=(Rotation=(X=0,Y=0,Z=0.1305,W=0.9914),Translation=(X=0,Y=0,Z=-4),Scale3D=(X=1,Y=1,Z=1))))
    WeightMinimum=0
    Weight=@n6.OutputParam
   )
   > &RigUnit_ModifyTransforms(
    ItemToModify=((Item=(Type=Control,Name="dyn_chain_front_8_ctrl"),Transform=(Rotation=(X=0,Y=0,Z=0,W=1),Translation=(X=0,Y=2,Z=0),Scale3D=(X=1,Y=1,Z=1))))
    WeightMinimum=0
    Weight=@n6.OutputParam
   )
   > &RigUnit_ModifyTransforms(
    ItemToModify=((Item=(Type=Control,Name="dyn_chain_back_1_ctrl"),Transform=(Rotation=(X=0,Y=0,Z=0,W=1),Translation=(X=0,Y=0,Z=0.5),Scale3D=(X=1,Y=1,Z=1))))
    WeightMinimum=0
    Weight=@n6.OutputParam
   )
   > &RigUnit_ModifyTransforms(
    ItemToModify=((Item=(Type=Control,Name="dyn_chain_back_4_ctrl"),Transform=(Rotation=(X=0,Y=0,Z=0,W=1),Translation=(X=0,Y=2,Z=-2),Scale3D=(X=1,Y=1,Z=1))))
    WeightMinimum=0
    Weight=@n6.OutputParam
   )
   > &RigUnit_ModifyTransforms(
    ItemToModify=((Item=(Type=Control,Name="dyn_chain_back_7_ctrl"),Transform=(Rotation=(X=0,Y=0,Z=0,W=1),Translation=(X=-1,Y=-3,Z=-1),Scale3D=(X=1,Y=1,Z=1))))
    WeightMinimum=0
    Weight=@n6.OutputParam
   )
   > &RigUnit_ModifyTransforms(
    ItemToModify=((Item=(Type=Control,Name="dyn_chain_back_9_ctrl"),Transform=(Rotation=(X=0,Y=0,Z=0,W=1),Translation=(X=0,Y=2,Z=1),Scale3D=(X=1,Y=1,Z=1))))
    WeightMinimum=0
    Weight=@n6.OutputParam
   )
   > &RigUnit_ModifyTransforms(
    ItemToModify=((Item=(Type=Bone,Name="dyn_tshirt_front"),Transform=(Rotation=(X=0,Y=-0.1736,Z=0,W=0.9848),Translation=(X=-3,Y=1,Z=0),Scale3D=(X=1,Y=1,Z=1))))
    Weight=1
    WeightMinimum=0
   )
   RerouteNode_11_1 = RigVMDispatchNode(Values="((Type=Bone,Name=\"dyn_chain_front_1\"),((),Type=Bone,Name=\"dyn_chain_front_2\"),((),Type=Bone,Name=\"dyn_chain_front_3\"),((),Type=Bone,Name=\"dyn_chain_front_4\"),((),Type=Bone,Name=\"dyn_chain_front_5\"),((),Type=Bone,Name=\"dyn_chain_front_6\"),((),Type=Bone,Name ...(truncated)") {template="DISPATCH_RigVMDispatch_ArrayMake(in Values,out Array)"}
   RerouteNode_15_1 = RigVMDispatchNode(Values=((Type=Control,Name="dyn_chain_front_1_ctrl"),((),Type=Control,Name="dyn_chain_front_4_ctrl"),((),Type=Control,Name="dyn_chain_front_6_ctrl"),((),Type=Control,Name="dyn_chain_front_8_ctrl"))) {template="DISPATCH_RigVMDispatch_ArrayMake(in Values,out Array)"}
   > SplineIK_2_2 = RigVMFunctionReferenceNode(
    Primary Axis=(X=1,Y=0,Z=0)
    Samples Per Segment=16
    Secondary Spline Direction=(X=0,Y=1,Z=0)
    Spline Mode=BSpline
    Stretch=true
    Up Axis=(X=0,Y=0,Z=0)
    Bones=@RerouteNode_11_1.Array
    Controls=@RerouteNode_15_1.Array
   )
   RerouteNode_12_1 = RigVMDispatchNode(Values="((Type=Bone,Name=\"dyn_chain_back_1\"),((),Type=Bone,Name=\"dyn_chain_back_2\"),((),Type=Bone,Name=\"dyn_chain_back_3\"),((),Type=Bone,Name=\"dyn_chain_back_4\"),((),Type=Bone,Name=\"dyn_chain_back_5\"),((),Type=Bone,Name=\"dyn_chain_back_6\"),((),Type=Bone,Name=\"dyn_ ...(truncated)") {template="DISPATCH_RigVMDispatch_ArrayMake(in Values,out Array)"}
   RerouteNode_16_1 = RigVMDispatchNode(Values=((Type=Control,Name="dyn_chain_back_1_ctrl"),((),Type=Control,Name="dyn_chain_back_4_ctrl"),((),Type=Control,Name="dyn_chain_back_7_ctrl"),((),Type=Control,Name="dyn_chain_back_9_ctrl"),((),Type=Control,Name="dyn_chain_back_9_ctrl"))) {template="DISPATCH_RigVMDispatch_ArrayMake(in Values,out Array)"}
   > SplineIK_2_1_1 = RigVMFunctionReferenceNode(
    Primary Axis=(X=1,Y=0,Z=0)
    Samples Per Segment=16
    Secondary Spline Direction=(X=0,Y=0.5,Z=0.5)
    Spline Mode=BSpline
    Stretch=true
    Up Axis=(X=0,Y=0,Z=0)
    Bones=@RerouteNode_12_1.Array
    Controls=@RerouteNode_16_1.Array
   )
   n7 = RigVMVariableNode() {variable=spr_thigh_r_up_alpha}
   n8 = RigVMVariableNode() {variable=SplineIK_chain_r_ctrls}
   /Script/ControlRig.RigUnit_HierarchySetPoseItemArray(
    ElementType=All
    Weight=&RigVMFunction_MathDoubleSub(A=1, B=@n7.Value).Result
    Pose=RigVMVariableNode() {variable=SplineIK_ctrls_r_bind_pose}.Value
    ItemsToSet=@n8.Value
    ExecutePin=@RigVMFunction_Sequence_3_1.C
   ) {is=struct RigVMUnitNode}
   > n9 = &RigUnit_HierarchySetPoseItemArray(ElementType=All, Pose=RigVMVariableNode() {variable=SplineIK_ctrl_thigh_r_up_pose}.Value, ItemsToSet=@n8.Value, Weight=@n7.Value)
   Branch_RigVMUnitNode_1_1 = &RigVMFunction_ControlFlowBranch(ExecuteContext=@n9.ExecutePin, Condition=RigVMVariableNode() {variable=is_front_end}.Value)
   n10 = RigVMVariableNode() {variable=spr_thigh_r_slight_back_out_apha}
   &RigUnit_HierarchySetPoseItemArray(
    ElementType=All
    ExecutePin=@Branch_RigVMUnitNode_1_1.True
    Weight=&RigVMFunction_MathDoubleSub(A=1, B=@n10.Value).Result
    Pose=RigVMVariableNode() {variable=SplineIK_ctrls_r_bind_pose}.Value
    ItemsToSet=@n8.Value
   )
   > &RigUnit_HierarchySetPoseItemArray(ElementType=All, Pose=(HierarchyTopologyVersion=-1,PoseHash=-1), Weight=@n10.Value, ItemsToSet=@n8.Value)
   RigVMFunction_Sequence_1_2 = RigVMAggregateNode(ExecuteContext=@RigVMFunction_Sequence_2_1.C)
   n11 = &RigUnit_SphericalPoseReader(
    ActiveRegionScaleFactors=(PositiveWidth=2,NegativeWidth=0.5,PositiveHeight=2,NegativeHeight=2)
    ActiveRegionSize=0.1
    Debug=(bDrawDebug=false,DebugScale=25,DebugSegments=20,DebugThickness=0.25)
    DriverAxis=(X=-1,Y=0,Z=0)
    DriverItem=(Type=Bone,Name="thigh_l")
    FalloffRegionScaleFactors=(PositiveWidth=1,NegativeWidth=1,PositiveHeight=1,NegativeHeight=1)
    FalloffSize=0.35
    FlipHeightScaling=false
    FlipWidthScaling=true
    OptionalParentItem=(Type=Bone)
    RotationOffset=(X=60,Y=180,Z=-17.5)
    ExecutePin=@RigVMFunction_Sequence_1_2.A
   )
   n12 = &RigVMFunction_MathBoolToFloat(Value=RigVMVariableNode() {variable=is_crouch}.Value)
   n13 = &RigVMFunction_MathDoubleMul(
    B=&RigVMFunction_MathDoubleClamp(Minimum=0.5, Value=/Script/RigVM.RigVMFunction_MathFloatMake(Value=@n12.Result) {is=struct RigVMUnitNode}.Value).Result
    A=@n11.OutputParam
   )
   > &RigUnit_ModifyTransforms(
    ItemToModify=((Item=(Type=Bone,Name="dyn_jacket_1_base"),Transform=(Rotation=(X=0.0436,Y=-0.001,Z=-0.0218,W=0.9988),Translation=(X=-5,Y=-2,Z=-3),Scale3D=(X=1,Y=1,Z=1))))
    WeightMinimum=0
    Weight=@n13.Result
   )
   > &RigUnit_ModifyTransforms(
    ItemToModify=((Item=(Type=Bone,Name="dyn_jacket_2_base"),Transform=(Rotation=(X=0,Y=0,Z=-0.0436,W=0.999),Translation=(X=-5,Y=-3,Z=0),Scale3D=(X=1,Y=1,Z=1))))
    WeightMinimum=0
    Weight=@n13.Result
   )
   > &RigUnit_ModifyTransforms(
    ItemToModify=((Item=(Type=Bone,Name="dyn_jacket_3_base"),Transform=(Rotation=(X=0,Y=0,Z=-0.0872,W=0.9962),Translation=(X=-2.5,Y=0,Z=0),Scale3D=(X=1,Y=1,Z=1))))
    WeightMinimum=0
    Weight=@n13.Result
   )
   n14 = &RigVMFunction_MathBoolToFloat(Value=RigVMVariableNode() {variable=is_crouch}.Value)
   > &RigUnit_ModifyTransforms(
    ItemToModify=((Item=(Type=Bone,Name="dyn_tshirt_front"),Transform=(Rotation=(X=0,Y=-0.1736,Z=0,W=0.9848),Translation=(X=-6,Y=2,Z=1),Scale3D=(X=1,Y=1,Z=1))))
    WeightMinimum=0
    Weight=&RigVMFunction_MathDoubleMul(B=&RigVMFunction_MathDoubleClamp(Minimum=0.25, Value=@n14.Result).Result, A=@n11.OutputParam).Result
   )
   n15 = &RigUnit_SphericalPoseReader(
    ActiveRegionScaleFactors=(PositiveWidth=2,NegativeWidth=0.5,PositiveHeight=2,NegativeHeight=2)
    ActiveRegionSize=0.1
    Debug=(bDrawDebug=false,DebugScale=25,DebugSegments=20,DebugThickness=0.25)
    DriverAxis=(X=1,Y=0,Z=0)
    DriverItem=(Type=Bone,Name="thigh_r")
    FalloffRegionScaleFactors=(PositiveWidth=1,NegativeWidth=1,PositiveHeight=1,NegativeHeight=1)
    FalloffSize=0.35
    FlipHeightScaling=false
    FlipWidthScaling=true
    OptionalParentItem=(Type=Bone)
    RotationOffset=(X=-75,Y=0,Z=-15)
    ExecutePin=@RigVMFunction_Sequence_1_2.B
   )
   n16 = &RigVMFunction_AlphaInterp(InterpSpeedDecreasing=50, InterpSpeedIncreasing=50, Scale=1, bClampResult=False, bInterpResult=False, Value=@n15.OutputParam)
   n17 = &RigVMFunction_MathBoolToFloat(Value=RigVMVariableNode() {variable=is_crouch}.Value)
   n18 = &RigVMFunction_MathDoubleMul(
    A=@n16.Result
    B=&RigVMFunction_MathDoubleClamp(Minimum=0.5, Value=&RigVMFunction_MathFloatMake(Value=@n17.Result).Value).Result
   )
   > &RigUnit_ModifyTransforms(
    ItemToModify=((Item=(Type=Bone,Name="dyn_jacket_9_base"),Transform=(Rotation=(X=-0.3894,Y=0.127,Z=0.1455,W=0.9006),Translation=(X=-12.5,Y=0,Z=0),Scale3D=(X=1,Y=1,Z=1))))
    WeightMinimum=0
    Weight=@n18.Result
   )
   > &RigUnit_ModifyTransforms(
    ItemToModify=((Item=(Type=Bone,Name="dyn_jacket_8_base"),Transform=(Rotation=(X=-0.171,Y=-0.0302,Z=0.171,W=0.9698),Translation=(X=7.5,Y=-1,Z=0),Scale3D=(X=1,Y=1,Z=1))))
    WeightMinimum=0.5
    Weight=@n18.Result
   )
   > &RigUnit_ModifyTransforms(
    ItemToModify=((Item=(Type=Bone,Name="dyn_jacket_7_base"),Transform=(Rotation=(X=0,Y=0,Z=0.0872,W=0.9962),Translation=(X=-1.5,Y=0,Z=-1),Scale3D=(X=1,Y=1,Z=1))))
    WeightMinimum=0.5
    Weight=@n18.Result
   )
   n19 = &RigVMFunction_MathDoubleSub(A=1, B=&RigVMFunction_MathBoolToFloat(Value=RigVMVariableNode() {variable=is_crouch}.Value).Result)
   > &RigUnit_ModifyTransforms(
    ItemToModify=((Item=(Type=Bone,Name="dyn_tshirt_front"),Transform=(Rotation=(X=0,Y=0,Z=0,W=1),Translation=(X=-5,Y=1,Z=0),Scale3D=(X=1,Y=1,Z=1))))
    WeightMinimum=0
    Weight=&RigVMFunction_MathDoubleMul(B=&RigVMFunction_MathDoubleClamp(Minimum=0.25, Value=@n19.Result).Result, A=@n18.Result).Result
   )
   n20 = &RigVMFunction_MathBoolToFloat(Value=RigVMVariableNode() {variable=is_fly}.Value)
   &RigUnit_ParentConstraint(
    AdvancedSettings=(RotationOrderForFilter=XZY)
    Child=(Type=Bone,Name="dyn_tshirt_back")
    Filter=(TranslationFilter=(bX=false,bY=false,bZ=false),RotationFilter=(bX=true,bY=true,bZ=true),ScaleFilter=(bX=True,bY=True,bZ=True))
    Parents=((Item=(Type=Bone,Name="pelvis"),Weight=1))
    Weight=@n20.Result
    ExecutePin=@RigVMFunction_Sequence_1_2.C
   )
   > &RigUnit_ModifyTransforms(
    ItemToModify=((Item=(Type=Bone,Name="dyn_tshirt_back"),Transform=(Rotation=(X=-0.0141,Y=0.0887,Z=-0.0849,W=0.9923),Translation=(X=-5,Y=0,Z=0),Scale3D=(X=1,Y=1,Z=1))))
    WeightMinimum=0
    Weight=@n20.Result
   )
   > &RigUnit_ModifyTransforms(
    ItemToModify=((Item=(Type=Bone,Name="dyn_jacket_5_base"),Transform=(Rotation=(X=0,Y=0,Z=0.0218,W=0.9998),Translation=(X=0,Y=0,Z=0),Scale3D=(X=1,Y=1,Z=1))))
    WeightMinimum=0
    Weight=@n20.Result
   )
   RigVMFunction_Sequence_1_1_1 = RigVMAggregateNode(ExecuteContext=@RigVMFunction_Sequence_1_2.D)
   n21 = &RigUnit_SphericalPoseReader(
    ActiveRegionScaleFactors=(PositiveWidth=1,NegativeWidth=2,PositiveHeight=1,NegativeHeight=2)
    ActiveRegionSize=0.1
    Debug=(bDrawDebug=false,DebugScale=25,DebugSegments=20,DebugThickness=0.25)
    DriverAxis=(X=1,Y=0,Z=0)
    DriverItem=(Type=Bone,Name="upperarm_l")
    FalloffRegionScaleFactors=(PositiveWidth=1,NegativeWidth=1,PositiveHeight=1,NegativeHeight=1)
    FalloffSize=0.45
    FlipHeightScaling=false
    FlipWidthScaling=true
    OptionalParentItem=(Type=Bone)
    RotationOffset=(X=60,Y=0,Z=-30)
    ExecutePin=@RigVMFunction_Sequence_1_1_1.A
   )
   n22 = &RigVMFunction_MathBoolToFloat(Value=RigVMVariableNode() {variable=is_fly}.Value)
   n23 = &RigVMFunction_MathDoubleClamp(Value=&RigVMFunction_MathDoubleAdd(B=&RigVMFunction_MathDoubleMul(B=-1, A=@n22.Result).Result, A=@n21.OutputParam).Result)
   > &RigUnit_ModifyTransforms(
    ItemToModify=((Item=(Type=Bone,Name="deform_upperarm_fr_base_l"),Transform=(Rotation=(X=0,Y=0,Z=0,W=1),Translation=(X=0,Y=1.5,Z=0),Scale3D=(X=1,Y=1,Z=1))))
    WeightMinimum=0
    Weight=@n23.Result
   )
   > &RigUnit_ModifyTransforms(
    ItemToModify=((Item=(Type=Bone,Name="dyn_vest_sleeve_l"),Transform=(Rotation=(X=0,Y=0,Z=0,W=1),Translation=(X=-5,Y=0,Z=2),Scale3D=(X=1,Y=1,Z=1))))
    WeightMinimum=0
    Weight=@n23.Result
   )
   > &RigUnit_ModifyTransforms(
    ItemToModify=((Item=(Type=Bone,Name="dyn_vest_sleeve_down_l"),Transform=(Rotation=(X=0.2588,Y=0,Z=0,W=0.9659),Translation=(X=0,Y=2,Z=7.5),Scale3D=(X=1,Y=1,Z=1))))
    WeightMinimum=0
    Weight=@n23.Result
   )
   > n24 = &RigUnit_ModifyTransforms(
    ItemToModify=((Item=(Type=Bone,Name="dyn_vest_sleeve_back_l"),Transform=(Rotation=(X=0,Y=0,Z=0,W=1),Translation=(X=0,Y=-5,Z=0),Scale3D=(X=1,Y=1,Z=1))))
    WeightMinimum=0
    Weight=@n22.Result
   )
   n25 = &RigVMFunction_Sequence(ExecuteContext=@n24.ExecutePin)
   &RigUnit_ModifyTransforms(
    ItemToModify=((Item=(Type=Bone,Name="dyn_badge_1"),Transform=(Rotation=(X=0.0725,Y=-0.081,Z=0.0923,W=0.9898),Translation=(X=-1.75,Y=-0.15,Z=0.25),Scale3D=(X=1,Y=1,Z=1))))
    WeightMinimum=0
    ExecutePin=@n25.A
    Weight=@n23.Result
   )
   > &RigUnit_ModifyTransforms(
    ItemToModify=((Item=(Type=Bone,Name="dyn_badge_2"),Transform=(Rotation=(X=0,Y=-0.0872,Z=0,W=0.9962),Translation=(X=-1,Y=0,Z=0),Scale3D=(X=1,Y=1,Z=1))))
    WeightMinimum=0
    Weight=@n23.Result
   )
   n26 = &RigVMFunction_MathDoubleSub(A=1, B=@n23.Result)
   &RigUnit_ModifyTransforms(
    ItemToModify=((Item=(Type=Bone,Name="dyn_badge_1"),Transform=(Rotation=(X=0,Y=0,Z=0.0872,W=0.9962),Translation=(X=0,Y=0.5,Z=0),Scale3D=(X=1,Y=1,Z=1))))
    WeightMinimum=0
    ExecutePin=@n25.B
    Weight=@n26.Result
   )
   > &RigUnit_ModifyTransforms(
    ItemToModify=((Item=(Type=Bone,Name="deform_upperarm_fr_base_l"),Transform=(Rotation=(X=0,Y=0,Z=0,W=1),Translation=(X=0,Y=2,Z=0),Scale3D=(X=1,Y=1,Z=1))))
    WeightMinimum=0
    Weight=@n26.Result
   )
   n27 = &RigUnit_SphericalPoseReader(
    ActiveRegionScaleFactors=(PositiveWidth=1,NegativeWidth=2,PositiveHeight=2,NegativeHeight=3)
    ActiveRegionSize=0.1
    Debug=(bDrawDebug=false,DebugScale=25,DebugSegments=20,DebugThickness=0.25)
    DriverAxis=(X=-1,Y=0,Z=0)
    DriverItem=(Type=Bone,Name="upperarm_r")
    FalloffRegionScaleFactors=(PositiveWidth=1,NegativeWidth=1,PositiveHeight=1,NegativeHeight=1)
    FalloffSize=0.45
    FlipHeightScaling=false
    FlipWidthScaling=true
    OptionalParentItem=(Type=Bone)
    RotationOffset=(X=-90,Y=45,Z=-30)
    ExecutePin=@RigVMFunction_Sequence_1_1_1.B
   )
   n28 = &RigVMFunction_MathBoolToFloat(Value=RigVMVariableNode() {variable=is_fly}.Value)
   n29 = &RigVMFunction_MathDoubleClamp(Value=&RigVMFunction_MathDoubleAdd(B=&RigVMFunction_MathDoubleMul(B=-1, A=@n28.Result).Result, A=@n27.OutputParam).Result)
   > &RigUnit_ModifyTransforms(
    ItemToModify=((Item=(Type=Bone,Name="dyn_vest_sleeve_r"),Transform=(Rotation=(X=0,Y=0.2588,Z=0,W=0.9659),Translation=(X=7.5,Y=0,Z=-2),Scale3D=(X=1,Y=1,Z=1))))
    WeightMinimum=0
    Weight=@n29.Result
   )
   > &RigUnit_ModifyTransforms(
    ItemToModify=((Item=(Type=Bone,Name="dyn_vest_sleeve_down_r"),Transform=(Rotation=(X=0.2588,Y=0,Z=0,W=0.9659),Translation=(X=0,Y=-2,Z=-7.5),Scale3D=(X=1,Y=1,Z=1))))
    WeightMinimum=0
    Weight=@n29.Result
   )
   > &RigUnit_ModifyTransforms(
    ItemToModify=((Item=(Type=Bone,Name="dyn_vest_sleeve_back_r"),Transform=(Rotation=(X=0,Y=0,Z=0,W=1),Translation=(X=0,Y=5,Z=0),Scale3D=(X=1,Y=1,Z=1))))
    WeightMinimum=0
    Weight=@n28.Result
   )
   n30 = &RigUnit_SphericalPoseReader(
    ActiveRegionScaleFactors=(PositiveWidth=0.2,NegativeWidth=0.2,PositiveHeight=0.2,NegativeHeight=0.2)
    ActiveRegionSize=0.1
    Debug=(bDrawDebug=false,DebugScale=25,DebugSegments=20,DebugThickness=0.25)
    DriverAxis=(X=0,Y=0,Z=1)
    DriverItem=(Type=Bone,Name="spine_01")
    FalloffRegionScaleFactors=(PositiveWidth=1,NegativeWidth=1,PositiveHeight=1,NegativeHeight=1)
    FalloffSize=0.1
    FlipHeightScaling=false
    FlipWidthScaling=true
    OptionalParentItem=(Type=Bone)
    RotationOffset=(X=0,Y=-2.5,Z=0)
    ExecutePin=@RigVMFunction_Sequence_1_1_1.C
   )
   n31 = &RigVMFunction_MathDoubleSub(A=1, B=@n30.OutputParam)
   > &RigUnit_ModifyTransforms(
    ItemToModify=((Item=(Type=Bone,Name="dyn_tshirt_back"),Transform=(Rotation=(X=0,Y=0,Z=0,W=1),Translation=(X=-5,Y=0,Z=-5),Scale3D=(X=1,Y=1,Z=1))))
    WeightMinimum=0
    Weight=@n31.Result
   )
   n32 = &RigVMFunction_MathBoolOr(A=RigVMVariableNode() {variable=is_crouch}.Value, B=RigVMVariableNode() {variable=is_DBNO}.Value)
   > &RigUnit_ModifyTransforms(
    ItemToModify=((Item=(Type=Bone,Name="dyn_tshirt_back"),Transform=(Rotation=(X=0.0118,Y=0.0473,Z=0.2153,W=0.9753),Translation=(X=0,Y=0,Z=0),Scale3D=(X=1,Y=1,Z=1))))
    WeightMinimum=0
    ItemToModify.0.Transform.Translation.Y=&RigVMFunction_MathDoubleClamp(Maximum=0, Minimum=-4, Value=&RigVMFunction_MathDoubleMul(A=-10, B=@n31.Result).Result).Result
    ItemToModify.0.Transform.Translation.Z=&RigVMFunction_MathDoubleMul(A=1, B=@n31.Result).Result
    Weight=&RigVMFunction_MathFloatMake(Value=&RigVMFunction_MathBoolToFloat(Value=@n32.Result).Result).Value
   )
   RigVMFunction_MathBoolToFloat = &RigVMFunction_MathBoolToFloat(Value=RigVMVariableNode() {variable=is_front_end}.Value)
   &RigUnit_ModifyTransforms(
    ItemToModify=((Item=(Type=Bone,Name="dyn_tshirt_front"),Transform=(Rotation=(X=0.1305,Y=0,Z=0,W=0.9914),Translation=(X=0,Y=1,Z=-0.5),Scale3D=(X=1,Y=1,Z=1))))
    Weight=@RigVMFunction_MathBoolToFloat.Result
    ExecutePin=@RigVMFunction_Sequence.B
   )
   &RigVMFunction_MathDoubleClamp(Minimum=0.55, Value=@n1.OutputParam)
   &RigUnit_ModifyTransforms(
    ExecutePin=()
    ItemToModify=((Item=(Type=Bone,Name="dyn_chain_control_1"),Transform=(Rotation=(X=0,Y=0,Z=0,W=1),Translation=(X=-2.5,Y=0,Z=0),Scale3D=(X=1,Y=1,Z=1))))
    Weight=1
    WeightMinimum=0
    ItemToModify.0.Transform.Rotation=&RigVMFunction_MathQuaternionFromEuler(Euler=(X=15,Y=0,Z=20)).Result
   )
  ]}
 RigVMModel Construction Event Graph {nodes [
   &RigUnit_ModifyTransforms(
    ExecutePin=()
    ItemToModify=((Item=(Type=Control,Name="dyn_chain_end_r_ctrl"),Transform=(Rotation=(X=0,Y=0,Z=0,W=1),Translation=(X=1,Y=-0.1,Z=0),Scale3D=(X=1,Y=1,Z=1))))
    Weight=1
   )
   > &RigUnit_ModifyTransforms(
    ItemToModify=((Item=(Type=Control,Name="dyn_front_chain_03_r_ctrl"),Transform=(Rotation=(X=0,Y=0,Z=0,W=1),Translation=(X=0,Y=0,Z=0),Scale3D=(X=1,Y=1,Z=1))))
    Weight=1
   )
   > &RigUnit_ModifyTransforms(
    ItemToModify=((Item=(Type=Control,Name="dyn_front_chain_05_r_ctrl"),Transform=(Rotation=(X=0,Y=0,Z=0,W=1),Translation=(X=0,Y=0,Z=0),Scale3D=(X=1,Y=1,Z=1))))
    Weight=1
   )
   n1 = &RigUnit_PrepareForExecution()
   SequenceAggregate = &RigVMFunction_Sequence(ExecuteContext=@n1.ExecutePin)
   SequenceAggregate_1 = &RigVMFunction_Sequence(ExecuteContext=@SequenceAggregate.A)
   RerouteNode_13_1 = RigVMDispatchNode(Values="((Type=Control,Name=\"dyn_chain_front_1_ctrl\"),((),Type=Control,Name=\"dyn_chain_front_4_ctrl\"),((),Type=Control,Name=\"dyn_chain_front_6_ctrl\"),((),Type=Control,Name=\"dyn_chain_front_8_ctrl\"),((),Type=Control,Name=\"dyn_chain_back_9_ctrl\"),((),Type=Control,Na ...(truncated)") {template="DISPATCH_RigVMDispatch_ArrayMake(in Values,out Array)"}
   RigVMVariableNode(Value=@RerouteNode_13_1.Array, ExecuteContext=@SequenceAggregate_1.A) {variable=SplineIK_chain_r_ctrls}
   RigVMVariableNode(
    Value=/Script/ControlRig.RigUnit_HierarchyGetPoseItemArray(ElementType=All, ItemsToGet=RigVMVariableNode() {variable=SplineIK_chain_r_ctrls}.Value) {is=struct RigVMUnitNode}.Pose
    ExecuteContext=@SequenceAggregate_1.B
   ) {variable=SplineIK_ctrls_r_bind_pose}
   SequenceAggregate_2 = &RigVMFunction_Sequence(ExecuteContext=@SequenceAggregate.B)
   n2 = RigVMVariableNode() {variable=spr_thigh_r_up_alpha}
   &RigUnit_ModifyTransforms(
    ItemToModify=((Item=(Type=Control,Name="dyn_chain_front_4_ctrl"),Transform=(Rotation=(X=0,Y=0,Z=0,W=1),Translation=(X=0,Y=-5,Z=-2),Scale3D=(X=1,Y=1,Z=1))))
    ExecutePin=@SequenceAggregate_2.A
    Weight=@n2.Value
   )
   > &RigUnit_ModifyTransforms(
    ItemToModify=((Item=(Type=Control,Name="dyn_chain_front_6_ctrl"),Transform=(Rotation=(X=0,Y=0,Z=0,W=1),Translation=(X=-1,Y=3,Z=-2),Scale3D=(X=1,Y=1,Z=1))))
    Weight=@n2.Value
   )
   > &RigUnit_ModifyTransforms(
    ItemToModify=((Item=(Type=Control,Name="dyn_chain_back_4_ctrl"),Transform=(Rotation=(X=0,Y=0,Z=0,W=1),Translation=(X=0,Y=0,Z=0),Scale3D=(X=1,Y=1,Z=1))))
    Weight=@n2.Value
   )
   > &RigUnit_ModifyTransforms(
    ItemToModify=((Item=(Type=Control,Name="dyn_chain_back_7_ctrl"),Transform=(Rotation=(X=0,Y=0,Z=0,W=1),Translation=(X=0,Y=0,Z=0),Scale3D=(X=1,Y=1,Z=1))))
    Weight=@n2.Value
   )
   > &RigUnit_ModifyTransforms(
    ItemToModify=((Item=(Type=Control,Name="dyn_chain_front_Main_ctrl"),Transform=(Rotation=(X=-0.25,Y=-0.25,Z=-0.067,W=0.933),Translation=(X=-3,Y=0,Z=0),Scale3D=(X=1,Y=1,Z=1))))
    Weight=@n2.Value
   )
   > RigVMVariableNode(Value=&RigUnit_HierarchyGetPoseItemArray(ElementType=All, ItemsToGet=RigVMVariableNode() {variable=SplineIK_chain_r_ctrls}.Value).Pose) {variable=SplineIK_ctrl_thigh_r_up_pose}
   &RigUnit_HierarchySetPoseItemArray(
    ElementType=All
    Weight=1
    ExecutePin=@SequenceAggregate_2.B
    Pose=RigVMVariableNode() {variable=SplineIK_ctrls_r_bind_pose}.Value
    ItemsToSet=RigVMVariableNode() {variable=SplineIK_chain_r_ctrls}.Value
   )
   RerouteNode_4_ArrayMake = RigVMDispatchNode(Values=((Type=None,Name="None"))) {template="DISPATCH_RigVMDispatch_ArrayMake(in Values,out Array)"}
   &RigUnit_HierarchyGetPoseItemArray(ElementType=All, ItemsToGet=@RerouteNode_4_ArrayMake.Array)
   &RigUnit_ModifyTransforms(
    ExecutePin=()
    ItemToModify=((Item=(Type=Control,Name="dyn_chain_front_6_ctrl"),Transform=(Rotation=(X=0,Y=0,Z=0,W=1),Translation=(X=-1,Y=3,Z=-2),Scale3D=(X=1,Y=1,Z=1))))
    Weight=1
   )
   &RigUnit_ModifyTransforms(
    ExecutePin=()
    ItemToModify=((Item=(Type=Control,Name="dyn_chain_front_4_ctrl"),Transform=(Rotation=(X=0,Y=0,Z=0,W=1),Translation=(X=0,Y=-5,Z=-2),Scale3D=(X=1,Y=1,Z=1))))
    Weight=0
   )
   &RigUnit_ModifyTransforms(
    ExecutePin=()
    ItemToModify=((Item=(Type=Control,Name="dyn_chain_front_1_ctrl"),Transform=(Rotation=(X=0,Y=0,Z=0,W=1),Translation=(X=0,Y=0,Z=10),Scale3D=(X=1,Y=1,Z=1))))
    Weight=1
   )
   RigVMFunction_MathBoolToFloat = &RigVMFunction_MathBoolToFloat(Value=RigVMVariableNode() {variable=is_front_end}.Value)
  ]}
]
hierarchy [
 root {loc=<0 0 0>, rot=<0 0 0 1>}
 pelvis {parent=root, loc=<0 -0.7071 93.1191>, rot=<0.0127 0.707 -0.0127 -0.707>}
 spine_01 {parent=pelvis, loc=<6.7494 -0.6388 0>, rot=<0 0 0.034 -0.9994>}
 spine_02 {parent=spine_01, loc=<7.252 -0.035 0>, rot=<0 0 0.0238 -0.9997>}
 spine_03 {parent=spine_02, loc=<6.0106 0 0>, rot=<0 0 0.0149 -0.9999>}
 spine_04 {parent=spine_03, loc=<8.1024 0 0>, rot=<0 0 -0.0132 -0.9999>}
 spine_05 {parent=spine_04, loc=<10.0142 0 0>, rot=<0 0 -0.0015 -1>}
 attach_cape {parent=spine_05, loc=<9.017 -11.714 0>, rot=<0.0616 -0.7044 -0.0616 -0.7044>}
 clavicle_l {parent=spine_05, loc=<4.3575 -0.1466 -2.3223>, rot=<0.0465 -0.7289 -0.009 -0.6829>}
 upperarm_l {parent=clavicle_l, loc=<15.5111 0 0>, rot=<0.0226 -0.3234 -0.0088 -0.9459>}
 lowerarm_l {parent=upperarm_l, loc=<26.6514 0 0>, rot=<0.0391 -0.0099 -0.2453 -0.9686>}
 hand_l {parent=lowerarm_l, loc=<22.3491 0 0>, rot=<-0.7081 -0.0507 0.0975 -0.6975>}
 thumb_01_l {parent=hand_l, loc=<1.3118 -0.3191 -2.459>, rot=<0.6546 -0.4042 0.0664 -0.6354>}
 thumb_02_l {parent=thumb_01_l, loc=<5.9336 -0.0448 0.0401>, rot=<-0.0037 0.0213 0.298 -0.9543>}
 thumb_03_l {parent=thumb_02_l, loc=<2.2223 0.3682 -0.0907>, rot=<0.0118 -0.013 0.0066 -0.9998>}
 index_metacarpal_l {parent=hand_l, loc=<1.6535 0.0183 -2.96>, rot=<0.0158 -0.0124 -0.0173 -0.9996>}
 index_01_l {parent=index_metacarpal_l, loc=<8.3008 0.0357 -0.0524>, rot=<-0.0044 0.0485 0.0924 -0.9945>}
 index_02_l {parent=index_01_l, loc=<3.3587 -0.3004 0.0452>, rot=<-0.002 0.0027 0.3225 -0.9466>}
 index_03_l {parent=index_02_l, loc=<2.5578 -0.123 -0.0252>, rot=<-0.002 0.0067 0.2558 -0.9667>}
 middle_metacarpal_l {parent=hand_l, loc=<2.1856 -0.1725 -0.5984>, rot=<0.006 0.0168 -0.0329 -0.9993>}
 middle_01_l {parent=middle_metacarpal_l, loc=<7.3275 -0.2447 0.0129>, rot=<-0.0208 0.0042 0.0858 -0.9961>}
 middle_02_l {parent=middle_01_l, loc=<4.0123 0.1401 0.1175>, rot=<-0.0218 0.0258 0.3512 -0.9357>}
 middle_03_l {parent=middle_02_l, loc=<2.6077 0.0534 0.1373>, rot=<0.0082 0.0188 0.1708 -0.9851>}
 ring_metacarpal_l {parent=hand_l, loc=<2.2074 -0.3625 1.3638>, rot=<-0.0284 0.0722 -0.0458 -0.9959>}
 ring_01_l {parent=ring_metacarpal_l, loc=<7.0295 -0.6197 -0.1446>, rot=<-0.0489 -0.0252 0.1892 -0.9804>}
 ring_02_l {parent=ring_01_l, loc=<3.6174 0.4647 -0.1027>, rot=<0 0 0.2806 -0.9598>}
 ring_03_l {parent=ring_02_l, loc=<2.4152 0.114 0.0199>, rot=<0 0 0.1189 -0.9929>}
 pinky_metacarpal_l {parent=hand_l, loc=<1.7616 -0.9753 3.3814>, rot=<-0.0397 0.0814 -0.069 -0.9935>}
 pinky_01_l {parent=pinky_metacarpal_l, loc=<7.1709 -0.4831 -0.1094>, rot=<-0.0654 -0.0035 0.2242 -0.9723>}
 pinky_02_l {parent=pinky_01_l, loc=<2.8274 0.0893 -0.1043>, rot=<0.0037 0.0011 0.2854 -0.9584>}
 pinky_03_l {parent=pinky_02_l, loc=<2.2046 0.0307 -0.0078>, rot=<-0.0002 0.0098 0.1445 -0.9895>}
 deform_wrist_dn_base_l {parent=hand_l, loc=<0 -2.2342 0>, rot=<0 0 0 -1>}
 deform_wrist_dn_l {parent=deform_wrist_dn_base_l, loc=<0 0 0>, rot=<0 0 0 1>}
 deform_wrist_up_base_l {parent=hand_l, loc=<0 2.2682 0>, rot=<0 0 0 -1>}
 deform_wrist_up_l {parent=deform_wrist_up_base_l, loc=<0 0 0>, rot=<0 0 0 1>}
 lowerarm_twist_02_l {parent=lowerarm_l, loc=<9.6199 0 0>, rot=<0 0 0 -1>}
 lowerarm_twist_01_l {parent=lowerarm_l, loc=<17.8549 0 0>, rot=<0 0 0 -1>}
 deform_elbow_in_base_l {parent=lowerarm_l, loc=<0.2112 5.259 0.8274>, rot=<0 0 0 -1>}
 deform_elbow_in_l {parent=deform_elbow_in_base_l, loc=<0 0 0>, rot=<0 0 0 1>}
 deform_elbow_out_base_l {parent=lowerarm_l, loc=<0 -4 0>, rot=<0 0 0 -1>}
 deform_elbow_out_l {parent=deform_elbow_out_base_l, loc=<0 0 0>, rot=<0 0 0 1>}
 upperarm_twist_01_l {parent=upperarm_l, loc=<0.0041 0 0>, rot=<0 0 0 -1>}
 deform_upperarm_bk_base_l {parent=upperarm_twist_01_l, loc=<1.0813 -7.2193 -1.9453>, rot=<0 0 0 -1>}
 deform_upperarm_bk_l {parent=deform_upperarm_bk_base_l, loc=<0 0 0>, rot=<0 0 0 1>}
 deform_upperarm_fr_base_l {parent=upperarm_twist_01_l, loc=<-0.3691 7.0506 -2.765>, rot=<0 0 0 -1>}
 deform_upperarm_fr_l {parent=deform_upperarm_fr_base_l, loc=<0 0 0>, rot=<0 0 0 1>}
 deform_upperarm_up_base_l {parent=upperarm_twist_01_l, loc=<-3.2216 -1.1391 3.3311>, rot=<0 0 0 -1>}
 deform_upperarm_up_l {parent=deform_upperarm_up_base_l, loc=<0 0 0>, rot=<0 0 0 1>}
 upperarm_twist_02_l {parent=upperarm_l, loc=<13.7941 0 0>, rot=<0 0 0 -1>}
 dyn_vest_sleeve_l {parent=clavicle_l, loc=<12.5089 7.2908 0.5076>, rot=<0.0008 0.0336 -0.0252 -0.9991>}
 dyn_vest_sleeve_back_l {parent=clavicle_l, loc=<14.5716 -8.751 -0.5932>, rot=<0.0008 0.0336 -0.0252 -0.9991>}
 attach_backpack {parent=spine_05, loc=<-0.123 -13.473 0>, rot=<-0.0093 -0.707 0.0093 -0.707>}
 neck_01 {parent=spine_05, loc=<11.3138 0.1529 0>, rot=<0 0 -0.2003 -0.9797>}
 neck_02 {parent=neck_01, loc=<4.5627 0 0>, rot=<0 0 0.0504 -0.9987>}
 head {parent=neck_02, loc=<5.3914 0 0>, rot=<0 0 0.115 -0.9934>}
 dyn_necklaces_1 {parent=spine_05, loc=<10.3781 5.816 0>, rot=<-0.2203 -0.6791 -0.6678 -0.2104>}
 dyn_necklaces_2 {parent=dyn_necklaces_1, loc=<4.1469 0 0>, rot=<-0.0009 -0.0718 -0.0129 -0.9973>}
 dyn_necklaces_3 {parent=dyn_necklaces_2, loc=<4.0446 0 0>, rot=<0 0 0 1>}
 clavicle_r {parent=spine_05, loc=<4.3575 -0.1466 2.3223>, rot=<0.7289 0.0465 -0.6829 0.009>}
 upperarm_r {parent=clavicle_r, loc=<-15.5111 0 0>, rot=<0.0226 -0.3234 -0.0088 -0.9459>}
 lowerarm_r {parent=upperarm_r, loc=<-26.6514 0 0>, rot=<0.0391 -0.0099 -0.2453 -0.9686>}
 hand_r {parent=lowerarm_r, loc=<-22.3491 0 0>, rot=<-0.7081 -0.0507 0.0975 -0.6975>}
 thumb_01_r {parent=hand_r, loc=<-1.3118 0.3191 2.459>, rot=<0.6546 -0.4042 0.0664 -0.6354>}
 thumb_02_r {parent=thumb_01_r, loc=<-5.9336 0.0448 -0.0401>, rot=<-0.0037 0.0213 0.298 -0.9543>}
 thumb_03_r {parent=thumb_02_r, loc=<-2.2223 -0.3682 0.0907>, rot=<0.0118 -0.013 0.0066 -0.9998>}
 index_metacarpal_r {parent=hand_r, loc=<-1.6535 -0.0183 2.96>, rot=<0.0158 -0.0124 -0.0173 -0.9996>}
 index_01_r {parent=index_metacarpal_r, loc=<-8.3008 -0.0357 0.0524>, rot=<-0.0044 0.0485 0.0924 -0.9945>}
 index_02_r {parent=index_01_r, loc=<-3.3587 0.3004 -0.0452>, rot=<-0.002 0.0027 0.3225 -0.9466>}
 index_03_r {parent=index_02_r, loc=<-2.5578 0.123 0.0252>, rot=<-0.002 0.0067 0.2558 -0.9667>}
 middle_metacarpal_r {parent=hand_r, loc=<-2.1856 0.1725 0.5984>, rot=<0.006 0.0168 -0.0329 -0.9993>}
 middle_01_r {parent=middle_metacarpal_r, loc=<-7.3275 0.2447 -0.0129>, rot=<-0.0208 0.0042 0.0858 -0.9961>}
 middle_02_r {parent=middle_01_r, loc=<-4.0123 -0.1401 -0.1175>, rot=<-0.0218 0.0258 0.3512 -0.9357>}
 middle_03_r {parent=middle_02_r, loc=<-2.6077 -0.0534 -0.1373>, rot=<0.0082 0.0188 0.1708 -0.9851>}
 ring_metacarpal_r {parent=hand_r, loc=<-2.2074 0.3625 -1.3638>, rot=<-0.0284 0.0722 -0.0458 -0.9959>}
 ring_01_r {parent=ring_metacarpal_r, loc=<-7.0295 0.6197 0.1446>, rot=<-0.0489 -0.0252 0.1892 -0.9804>}
 ring_02_r {parent=ring_01_r, loc=<-3.6174 -0.4647 0.1027>, rot=<0 0 0.2806 -0.9598>}
 ring_03_r {parent=ring_02_r, loc=<-2.4152 -0.114 -0.0199>, rot=<0 0 0.1189 -0.9929>}
 pinky_metacarpal_r {parent=hand_r, loc=<-1.7616 0.9753 -3.3814>, rot=<-0.0397 0.0814 -0.069 -0.9935>}
 pinky_01_r {parent=pinky_metacarpal_r, loc=<-7.1709 0.4831 0.1094>, rot=<-0.0654 -0.0035 0.2242 -0.9723>}
 pinky_02_r {parent=pinky_01_r, loc=<-2.8274 -0.0893 0.1043>, rot=<0.0037 0.0011 0.2854 -0.9584>}
 pinky_03_r {parent=pinky_02_r, loc=<-2.2046 -0.0307 0.0078>, rot=<-0.0002 0.0098 0.1445 -0.9895>}
 deform_wrist_dn_base_r {parent=hand_r, loc=<-0.0004 2.2345 -0.0001>, rot=<0 0 0 -1>}
 deform_wrist_dn_r {parent=deform_wrist_dn_base_r, loc=<0 0 0>, rot=<0 0 0 1>}
 deform_wrist_up_base_r {parent=hand_r, loc=<0.0003 -2.2684 0.0001>, rot=<0 0 0 -1>}
 deform_wrist_up_r {parent=deform_wrist_up_base_r, loc=<0 0 0>, rot=<0 0 0 1>}
 lowerarm_twist_02_r {parent=lowerarm_r, loc=<-9.6199 0 0>, rot=<0 0 0 -1>}
 lowerarm_twist_01_r {parent=lowerarm_r, loc=<-17.8549 0 0>, rot=<0 0 0 -1>}
 deform_elbow_in_base_r {parent=lowerarm_r, loc=<-0.2111 -5.2591 -0.8275>, rot=<0 0 0 -1>}
 deform_elbow_in_r {parent=deform_elbow_in_base_r, loc=<-0.0001 0 0.0002>, rot=<0 0 0 1>}
 deform_elbow_out_base_r {parent=lowerarm_r, loc=<0 4 -0.0001>, rot=<0 0 0 -1>}
 deform_elbow_out_r {parent=deform_elbow_out_base_r, loc=<0 0 0>, rot=<0 0 0 1>}
 upperarm_twist_01_r {parent=upperarm_r, loc=<-0.0041 0 0>, rot=<0 0 0 -1>}
 deform_upperarm_bk_base_r {parent=upperarm_twist_01_r, loc=<-1.0814 7.2192 1.9455>, rot=<0 0 0 1>}
 deform_upperarm_bk_r {parent=deform_upperarm_bk_base_r, loc=<0 0 0>, rot=<0 0 0 1>}
 deform_upperarm_fr_base_r {parent=upperarm_twist_01_r, loc=<0.3689 -7.0506 2.7652>, rot=<0 0 0 1>}
 deform_upperarm_fr_r {parent=deform_upperarm_fr_base_r, loc=<0 0 0>, rot=<0 0 0 1>}
 deform_upperarm_up_base_r {parent=upperarm_twist_01_r, loc=<3.2214 1.1391 -3.3308>, rot=<0 0 0 1>}
 deform_upperarm_up_r {parent=deform_upperarm_up_base_r, loc=<0 0 0>, rot=<0 0 0 -1>}
 upperarm_twist_02_r {parent=upperarm_r, loc=<-13.7941 0 0>, rot=<0 0 0 -1>}
 dyn_vest_sleeve_r {parent=clavicle_r, loc=<-12.5089 -7.2908 -0.5076>, rot=<0.0008 0.0336 -0.0252 -0.9991>}
 dyn_vest_sleeve_back_r {parent=clavicle_r, loc=<-14.5717 8.751 0.5935>, rot=<0.0008 0.0336 -0.0252 -0.9991>}
 dyn_badge_2 {parent=spine_05, loc=<1.0208 10.0594 -10.3962>, rot=<0.097 -0.6548 0.2101 -0.7195>}
 dyn_badge_1 {parent=spine_05, loc=<3.6961 7.5633 -13.2077>, rot=<0.0793 -0.6475 0.2495 -0.7157>}
 dyn_vest_sleeve_down_l {parent=spine_04, loc=<0.7132 -1.5441 -16.3621>, rot=<0.0294 -0.7065 -0.0294 -0.7065>}
 dyn_vest_sleeve_down_r {parent=spine_04, loc=<0.7128 -1.5442 16.3621>, rot=<0.7065 0.0294 -0.7065 0.0294>}
 dyn_jacket_9_base {parent=spine_03, loc=<3.8543 13.3621 6.1129>, rot=<0.0547 -0.9985 0 0>}
 dyn_jacket_9_1 {parent=dyn_jacket_9_base, loc=<0 0 0>, rot=<0 0 0 1>}
 dyn_jacket_9_2 {parent=dyn_jacket_9_1, loc=<5.609 0 0>, rot=<0 0 0 1>}
 dyn_jacket_9_3 {parent=dyn_jacket_9_2, loc=<7.816 0 0>, rot=<0 0 0 1>}
 dyn_jacket_8_base {parent=spine_03, loc=<4.4717 7.7473 13.9751>, rot=<0.0502 -0.9151 -0.3995 0.0219>}
 dyn_jacket_8_1 {parent=dyn_jacket_8_base, loc=<0 0 0>, rot=<0 0 0 1>}
 dyn_jacket_8_2 {parent=dyn_jacket_8_1, loc=<5.609 0 0>, rot=<0 0 0 1>}
 dyn_jacket_8_3 {parent=dyn_jacket_8_2, loc=<7.816 0 0>, rot=<0 0 0 1>}
 dyn_jacket_7_base {parent=spine_03, loc=<5.4965 -1.5712 16.0364>, rot=<0.0387 -0.706 -0.706 0.0387>}
 dyn_jacket_7_1 {parent=dyn_jacket_7_base, loc=<0 0 0>, rot=<0 0 0 1>}
 dyn_jacket_7_2 {parent=dyn_jacket_7_1, loc=<5.609 0 0>, rot=<0 0 0 1>}
 dyn_jacket_7_3 {parent=dyn_jacket_7_2, loc=<7.816 0 0>, rot=<0 0 0 1>}
 dyn_jacket_6_base {parent=spine_03, loc=<6.3416 -9.2563 10.4304>, rot=<-0.0165 0.3003 0.9523 -0.0522>}
 dyn_jacket_6_1 {parent=dyn_jacket_6_base, loc=<0 0 0>, rot=<0 0 0 1>}
 dyn_jacket_6_2 {parent=dyn_jacket_6_1, loc=<5.609 0 0>, rot=<0 0 0 1>}
 dyn_jacket_6_3 {parent=dyn_jacket_6_2, loc=<7.8159 0 0>, rot=<0 0 0 1>}
 dyn_jacket_5_base {parent=spine_03, loc=<6.5597 -11.24 0>, rot=<0 0 0.9985 -0.0547>}
 dyn_jacket_5_1 {parent=dyn_jacket_5_base, loc=<0 0 0>, rot=<0 0 0 1>}
 dyn_jacket_5_2 {parent=dyn_jacket_5_1, loc=<5.609 0 0>, rot=<0 0 0 1>}
 dyn_jacket_5_3 {parent=dyn_jacket_5_2, loc=<7.8159 0 0>, rot=<0 0 0 1>}
 dyn_jacket_4_base {parent=spine_03, loc=<6.3522 -9.3528 -9.901>, rot=<0.0165 -0.3003 0.9523 -0.0522>}
 dyn_jacket_4_1 {parent=dyn_jacket_4_base, loc=<0 0 0>, rot=<0 0 0 1>}
 dyn_jacket_4_2 {parent=dyn_jacket_4_1, loc=<5.609 0 0>, rot=<0 0 0 1>}
 dyn_jacket_4_3 {parent=dyn_jacket_4_2, loc=<7.8159 0 0>, rot=<0 0 0 1>}
 dyn_jacket_3_base {parent=spine_03, loc=<5.4965 -1.5712 -16.0364>, rot=<0.0387 -0.706 0.706 -0.0387>}
 dyn_jacket_3_1 {parent=dyn_jacket_3_base, loc=<0 0 0>, rot=<0 0 0 1>}
 dyn_jacket_3_2 {parent=dyn_jacket_3_1, loc=<5.609 0 0>, rot=<0 0 0 1>}
 dyn_jacket_3_3 {parent=dyn_jacket_3_2, loc=<7.8159 0 0>, rot=<0 0 0 1>}
 dyn_jacket_2_base {parent=spine_03, loc=<4.4718 7.7473 -13.9751>, rot=<0.0471 -0.8598 0.5077 -0.0278>}
 dyn_jacket_2_1 {parent=dyn_jacket_2_base, loc=<0 0 0>, rot=<0 0 0 1>}
 dyn_jacket_2_2 {parent=dyn_jacket_2_1, loc=<5.609 0 0>, rot=<0 0 0 1>}
 dyn_jacket_2_3 {parent=dyn_jacket_2_2, loc=<7.8159 0 0>, rot=<0 0 0 1>}
 dyn_jacket_1_base {parent=spine_03, loc=<3.8544 13.362 -6.1129>, rot=<0.0547 -0.9985 0 0>}
 dyn_jacket_1_1 {parent=dyn_jacket_1_base, loc=<0 0 0>, rot=<0 0 0 1>}
 dyn_jacket_1_2 {parent=dyn_jacket_1_1, loc=<5.609 0 0>, rot=<0 0 0 1>}
 dyn_jacket_1_3 {parent=dyn_jacket_1_2, loc=<7.8159 0 0>, rot=<0 0 0 1>}
 dyn_tshirt_back {parent=spine_01, loc=<2.6146 -11.0103 0>, rot=<0.1052 -0.9944 0.0001 -0.0007>}
 thigh_l {parent=pelvis, loc=<-2.5693 -0.0109 -7.7501>, rot=<0.0012 0.0572 0.0155 -0.9982>}
 calf_l {parent=thigh_l, loc=<-39.0484 0 0>, rot=<0.0006 -0.0161 -0.0391 -0.9991>}
 foot_l {parent=calf_l, loc=<-42.561 0.1077 0.4292>, rot=<0.0418 -0.0455 0.0343 -0.9975>}
 ball_l {parent=foot_l, loc=<-6.1599 13.8264 0.6806>, rot=<0.0001 -0.0001 -0.7063 -0.7079>}
 calf_twist_02_l {parent=calf_l, loc=<-18.8064 -0.0967 0.0397>, rot=<0.0237 0.0095 -0.0035 -0.9997>}
 calf_twist_01_l {parent=calf_l, loc=<-35.813 -0.0967 0.0397>, rot=<0.0237 0.0095 -0.0035 -0.9997>}
 deform_knee_in_base_l {parent=calf_l, loc=<1.298 -5.2957 -0.0496>, rot=<0 0 0 -1>}
 deform_knee_in_l {parent=deform_knee_in_base_l, loc=<0 0 0>, rot=<0 0 0 1>}
 deform_knee_out_base_l {parent=calf_l, loc=<0 4 0>, rot=<0 0 0 -1>}
 deform_knee_out_l {parent=deform_knee_out_base_l, loc=<0 0 0>, rot=<0 0 0 1>}
 dyn_shoelaces_2_base_l {parent=calf_l, loc=<-29.9573 5.0799 3.6213>, rot=<0.0508 0.0942 0.9938 0.0317>}
 dyn_shoelaces_2_1_l {parent=dyn_shoelaces_2_base_l, loc=<0 0 0>, rot=<0 0 0 1>}
 dyn_shoelaces_2_2_l {parent=dyn_shoelaces_2_1_l, loc=<5.747 0 0>, rot=<0 0 0 1>}
 dyn_shoelaces_1_base_l {parent=calf_l, loc=<-30.2115 3.4294 -5.244>, rot=<-0.0577 -0.9976 -0.0023 0.0369>}
 dyn_shoelaces_1_1_l {parent=dyn_shoelaces_1_base_l, loc=<0 0 0>, rot=<0 0 0 1>}
 dyn_shoelaces_1_2_l {parent=dyn_shoelaces_1_1_l, loc=<4.9834 0 0>, rot=<0 0 0 1>}
 thigh_twist_01_l {parent=thigh_l, loc=<-21.6303 0 0>, rot=<0.0013 0 0 -1>}
 deform_groin_base_l {parent=thigh_l, loc=<-2.1908 8.1142 -1.7442>, rot=<0 0 0 -1>}
 deform_groin_l {parent=deform_groin_base_l, loc=<0 0 0>, rot=<0 0 0 1>}
 dyn_zipper_base_l {parent=thigh_l, loc=<-23.756 -3.0777 -8.4913>, rot=<0.032 -0.9973 -0.0025 0.066>}
 dyn_zipper_l {parent=dyn_zipper_base_l, loc=<0 0 0>, rot=<0 0 0 1>}
 thigh_r {parent=pelvis, loc=<-2.5693 -0.0109 7.7501>, rot=<0.0572 -0.0012 0.9982 0.0155>}
 calf_r {parent=thigh_r, loc=<39.0484 0 0>, rot=<0.0006 -0.0161 -0.0391 -0.9991>}
 foot_r {parent=calf_r, loc=<42.561 -0.1077 -0.4292>, rot=<0.0418 -0.0455 0.0343 -0.9975>}
 ball_r {parent=foot_r, loc=<6.1599 -13.8264 -0.6806>, rot=<0.0001 -0.0001 -0.7063 -0.7079>}
 calf_twist_02_r {parent=calf_r, loc=<18.8064 0.0967 -0.0397>, rot=<0.0237 0.0095 -0.0035 -0.9997>}
 calf_twist_01_r {parent=calf_r, loc=<35.813 0.0967 -0.0397>, rot=<0.0237 0.0095 -0.0035 -0.9997>}
 deform_knee_in_base_r {parent=calf_r, loc=<-1.298 5.2957 0.0496>, rot=<0 0 0 -1>}
 deform_knee_in_r {parent=deform_knee_in_base_r, loc=<0 0 0>, rot=<0 0 0 1>}
 deform_knee_out_base_r {parent=calf_r, loc=<0 -4 0>, rot=<0 0 0 -1>}
 deform_knee_out_r {parent=deform_knee_out_base_r, loc=<0 0 0>, rot=<0 0 0 1>}
 dyn_shoelaces_2_base_r {parent=calf_r, loc=<29.9574 -5.0799 -3.6213>, rot=<0.0475 -0.0036 0.9982 0.0365>}
 dyn_shoelaces_2_1_r {parent=dyn_shoelaces_2_base_r, loc=<0 0 0>, rot=<0 0 0 1>}
 dyn_shoelaces_2_2_r {parent=dyn_shoelaces_2_1_r, loc=<-5.7469 0 0>, rot=<0 0 0 -1>}
 dyn_shoelaces_1_base_r {parent=calf_r, loc=<30.2115 -3.4294 5.244>, rot=<-0.0577 -0.9976 -0.0023 0.0369>}
 dyn_shoelaces_1_1_r {parent=dyn_shoelaces_1_base_r, loc=<0 0 0>, rot=<0 0 0 1>}
 dyn_shoelaces_1_2_r {parent=dyn_shoelaces_1_1_r, loc=<-4.9834 0 0>, rot=<0 0 0 1>}
 thigh_twist_01_r {parent=thigh_r, loc=<21.6303 0 0>, rot=<0.0013 0 0 -1>}
 deform_groin_base_r {parent=thigh_r, loc=<2.1908 -8.1142 1.7442>, rot=<0 0 0 -1>}
 deform_groin_r {parent=deform_groin_base_r, loc=<0 0 0>, rot=<0 0 0 1>}
 dyn_zipper_base_r {parent=thigh_r, loc=<23.7555 3.0757 8.4853>, rot=<0.032 -0.9973 -0.0025 0.066>}
 dyn_zipper_r {parent=dyn_zipper_base_r, loc=<0 0 0>, rot=<0 0 0 1>}
 dyn_chain_back_base {parent=pelvis, loc=<0.9099 -7.6375 11.2234>, rot=<-0.0127 -0.707 0.0127 -0.707>}
 dyn_chain_back_1 {parent=dyn_chain_back_base, loc=<0 0 0>, rot=<-0.0756 -0.8519 -0.0458 -0.5162>}
 dyn_chain_back_2 {parent=dyn_chain_back_1, loc=<1.5098 0.0441 0.0084>, rot=<0.4534 0.0577 -0.0294 -0.889>}
 dyn_chain_back_3 {parent=dyn_chain_back_2, loc=<1.8425 -0.0402 -0.1236>, rot=<-0.8587 -0.0324 -0.0069 -0.5114>}
 dyn_chain_back_4 {parent=dyn_chain_back_3, loc=<2.1468 0 0>, rot=<0.658 0.0186 -0.0163 -0.7526>}
 dyn_chain_back_5 {parent=dyn_chain_back_4, loc=<1.9378 0 0>, rot=<0.5559 0.0433 -0.029 -0.8296>}
 dyn_chain_back_6 {parent=dyn_chain_back_5, loc=<2.2369 0 0>, rot=<0.9798 0.012 0.0617 0.19>}
 dyn_chain_back_7 {parent=dyn_chain_back_6, loc=<2.0344 0 0>, rot=<0.0528 -0.0963 0.0051 -0.9939>}
 dyn_chain_back_8 {parent=dyn_chain_back_7, loc=<2.034 0 0>, rot=<0.9371 0.0318 -0.0887 -0.3361>}
 dyn_chain_back_9 {parent=dyn_chain_back_8, loc=<1.9486 0 0>, rot=<-0.0061 0.3399 0.0022 -0.9404>}
 dyn_chain_back_end {parent=dyn_chain_back_9, loc=<2.1636 0 0>, rot=<0 0 0 1>}
 dyn_chain_front_base {parent=pelvis, loc=<-1.1549 8.481 8.7132>, rot=<-0.0127 -0.707 0.0127 -0.707>}
 dyn_chain_front_1 {parent=dyn_chain_front_base, loc=<0 0 0>, rot=<0.0308 -0.8755 0.017 -0.482>}
 dyn_chain_front_2 {parent=dyn_chain_front_1, loc=<2.4188 0 0>, rot=<0.2712 -0.0505 0.0143 -0.9611>}
 dyn_chain_front_3 {parent=dyn_chain_front_2, loc=<2.3019 0 0>, rot=<0.3913 -0.0943 0.0403 -0.9145>}
 dyn_chain_front_4 {parent=dyn_chain_front_3, loc=<2.6512 0 0>, rot=<0.9355 -0.0047 -0.0124 0.353>}
 dyn_chain_front_5 {parent=dyn_chain_front_4, loc=<2.4649 0 0>, rot=<0.9424 -0.0528 0.1794 -0.2774>}
 dyn_chain_front_6 {parent=dyn_chain_front_5, loc=<2.2999 0 0>, rot=<-0.3327 -0.0679 -0.024 -0.9403>}
 dyn_chain_front_7 {parent=dyn_chain_front_6, loc=<2.09 0 0>, rot=<0.376 -0.2933 0.1268 -0.8697>}
 dyn_chain_front_8 {parent=dyn_chain_front_7, loc=<1.9575 0 0>, rot=<0 0 0 1>}
 deform_glute_base_l {parent=pelvis, loc=<-8.8064 -8.8222 -8.4287>, rot=<0.0275 0.0501 -0.4472 -0.8926>}
 deform_glute_l {parent=deform_glute_base_l, loc=<0 0 0>, rot=<0 0 0 1>}
 deform_glute_base_r {parent=pelvis, loc=<-8.8062 -8.8222 8.4287>, rot=<0.0501 -0.0275 0.8926 -0.4472>}
 deform_glute_r {parent=deform_glute_base_r, loc=<0 0 0>, rot=<0 0 0 1>}
 dyn_chain_control_1 {parent=pelvis, loc=<-0.1225 0.4217 9.9683>, rot=<-0.02 -0.9543 0.1049 0.2791>}
 dyn_chain_control_2 {parent=dyn_chain_control_1, loc=<5.7075 0 0>, rot=<0 0 0 1>}
 dyn_chain_control_3 {parent=dyn_chain_control_2, loc=<7.4616 0 0>, rot=<0 0 0 1>}
 dyn_tshirt_front {parent=pelvis, loc=<6.5661 9.5594 0>, rot=<-0.0429 -0.9991 0 -0.0042>}
 dyn_chain_front_1_null {type=Null, loc=<-8.7132 7.727 91.6611>, rot=<0.0308 -0.8755 0.017 -0.482>}
 dyn_chain_front_1_ctrl {controlType=EULER_TRANSFORM, parent=dyn_chain_front_1_null, type=Control, loc=<0 0 0>, rot=<0 0 0 1>}
 dyn_chain_front_4_null {type=Null, loc=<-13.0163 6.423 85.892>, rot=<-0.0809 -0.7267 0.5201 -0.4414>}
 dyn_chain_front_4_ctrl {controlType=EULER_TRANSFORM, parent=dyn_chain_front_4_null, type=Control, loc=<0 0 0>, rot=<0 0 0 1>}
 dyn_chain_front_4_ccorr_null {parent=dyn_chain_front_4_null, type=Null, loc=<0 0 0>, rot=<0 0 0 1>}
 dyn_chain_front_6_null {type=Null, loc=<-15.6342 4.0591 82.8148>, rot=<0.4585 -0.8574 -0.163 -0.1675>}
 dyn_chain_front_6_ctrl {controlType=EULER_TRANSFORM, parent=dyn_chain_front_6_null, type=Control, loc=<0 0 0>, rot=<0 0 0 1>}
 dyn_chain_front_Main_null {type=Null, loc=<-16.9289 0.5838 81.8365>, rot=<0.6183 -0.6754 -0.3084 0.2576>}
 dyn_chain_front_Main_ctrl {controlType=EULER_TRANSFORM, parent=dyn_chain_front_Main_null, type=Control, loc=<0 0 0>, rot=<0 0 0 1>}
 dyn_chain_front_8_ctrl {controlType=EULER_TRANSFORM, parent=dyn_chain_front_Main_ctrl, type=Control, loc=<0 0 0>, rot=<0 0 0 1>}
 dyn_chain_back_9_ctrl {controlType=EULER_TRANSFORM, parent=dyn_chain_front_Main_ctrl, type=Control, loc=<0 0 0>, rot=<0 0 0 1>}
 dyn_chain_back_7_null {type=Null, loc=<-16.0344 -4.1287 83.9405>, rot=<0.5741 0.728 -0.3586 0.1088>}
 dyn_chain_back_7_ctrl {controlType=EULER_TRANSFORM, parent=dyn_chain_back_7_null, type=Control, loc=<0 0 0>, rot=<0 0 0 1>}
 dyn_chain_back_4_null {type=Null, loc=<-13.6953 -6.8837 89.1905>, rot=<-0.0571 0.7349 0.3714 0.5645>}
 dyn_chain_back_4_ctrl {controlType=EULER_TRANSFORM, parent=dyn_chain_back_4_null, type=Control, loc=<0 0 0>, rot=<0 0 0 1>}
 dyn_chain_back_1_null {type=Null, loc=<-11.2234 -8.3072 94.302>, rot=<-0.0756 -0.8519 -0.0458 -0.5162>}
 dyn_chain_back_1_ctrl {controlType=EULER_TRANSFORM, parent=dyn_chain_back_1_null, type=Control, loc=<0 0 0>, rot=<0 0 0 1>}
]
variables [
 SplineIK_chain_r_ctrls {
  type="TArray<FRigElementKey>"
  default="((Type=Control,Name=\"dyn_front_chain_01_r_ctrl\"),(Type=Control,Name=\"dyn_front_chain_03_r_ctrl\"),(Type=Control,Name=\"dyn_front_chain_05_r_ctrl\"),(Type=Control,Name=\"dyn_front_chain_07_r_ctrl\"),(Type=Control,Name=\"dyn_back_chain_01_r_ctrl\"),(Type=Control,Na ...(truncated)"
 }
 SplineIK_ctrl_thigh_r_up_pose {type=FRigPose, default=(Elements=,HierarchyTopologyVersion=-1,PoseHash=-1)}
 SplineIK_ctrls_r_bind_pose {type=FRigPose, default=(Elements=,HierarchyTopologyVersion=-1,PoseHash=-1)}
 is_DBNO {type=bool, default=False}
 is_crouch {type=bool, default=False}
 is_fly {type=bool, default=False}
 is_front_end {type=bool, default=False}
 spr_thigh_r_slight_back_out_apha {type=double, default=0}
 spr_thigh_r_up_alpha {type=double, default=0}
]
