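# Physics asset for the M_MED_KelpLinenCalcium male player body cosmetic, providing capsule bodies and joints for its jacket, necklaces and chain.
# Capsules on pelvis, spine and thighs form the body volume beside the dynamic bones.
# Nine four-body jacket strands use tight angular limits, with swing locked at both end strands' tips.
# Necklace joints to spine_05 are angularly locked, while the three-body chain swings more widely and ends in a three-capsule body.
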PhysicsAsset /BRCosmetics/Characters/Player/Male/Medium/Bodies/M_MED_KelpLinenCalcium/Meshes/M_MED_KelpLinenCalcium_Physics
bodies {
 dyn_chain_control_1 [{shapeType=Capsule, length=0, radius=2.75, shapeName=dyn_chain_control_1_capsule, rotation=<rot -0.0001 0.0002 90.0001>}]
 dyn_chain_control_2 [{
   shapeType=Capsule
   length=17.5
   radius=2.5
   shapeName=dyn_chain_control_2_capsule
   center=<-0.2774 -0.554 -2.8173>
   rotation=<rot -0.8674 -9.9613 95.0766>
  }]
 dyn_chain_control_3 [
  {
   shapeType=Capsule
   length=10.5
   radius=3.25
   shapeName=dyn_chain_control_3_capsule
   center=<-3.2759 4.5675 -1.2083>
   rotation=<rot 12.9069 62.8222 89.4423>
  }
  {
   shapeType=Capsule
   length=9
   radius=3.25
   shapeName=dyn_chain_control_3_capsule
   center=<-5.821 -6.6795 -0.2902>
   rotation=<rot 20.3423 -67.4027 -82.9801>
  }
  {
   shapeType=Capsule
   length=5.5
   radius=3.5
   shapeName=dyn_chain_control_3_capsule
   center=<-0.4961 0 -1.6711>
   rotation=<rot 30 0.0002 -89.9999>
  }
 ]
 dyn_jacket_1_1 [{shapeType=Capsule, length=7.75, radius=2.25, shapeName=dyn_jacket_1_1_capsule, center=<4.0782 0 0>, rotation=<rot 0 0.0003 0>}]
 dyn_jacket_1_2 [{shapeType=Capsule, length=7.5, radius=2.5, shapeName=dyn_jacket_1_2_capsule, center=<4.0782 0 0>, rotation=<rot 0 0.0003 0>}]
 dyn_jacket_1_3 [{
   shapeType=Capsule
   length=8.25
   radius=2
   shapeName=dyn_jacket_1_3_capsule
   center=<5.7128 -0.0476 0.3396>
   rotation=<rot 4.094 -2.8719 -25.1027>
  }]
 dyn_jacket_1_base [{shapeType=Capsule, length=0, radius=2.25, shapeName=dyn_jacket_1_base_capsule}]
 dyn_jacket_2_1 [{shapeType=Capsule, length=7.75, radius=2.25, shapeName=dyn_jacket_2_1_capsule, center=<4.0782 0 0>, rotation=<rot -0.0001 0.0002 0>}]
 dyn_jacket_2_2 [{shapeType=Capsule, length=7.5, radius=2.5, shapeName=dyn_jacket_2_2_capsule, center=<4.0782 0 0>, rotation=<rot -0.0001 0.0002 0>}]
 dyn_jacket_2_3 [{shapeType=Capsule, length=8, radius=2, shapeName=dyn_jacket_2_3_capsule, center=<5.0178 -0.5389 0>, rotation=<rot -0.0003 0.0003 0>}]
 dyn_jacket_2_base [{shapeType=Capsule, length=0, radius=2.25, shapeName=dyn_jacket_2_base_capsule}]
 dyn_jacket_3_1 [{shapeType=Capsule, length=7.75, radius=2.25, shapeName=dyn_jacket_3_1_capsule, center=<4.0782 0 0>, rotation=<rot -0.0001 0 0>}]
 dyn_jacket_3_2 [{shapeType=Capsule, length=7.5, radius=2.5, shapeName=dyn_jacket_3_2_capsule, center=<4.0782 0 0>, rotation=<rot -0.0001 0 0>}]
 dyn_jacket_3_3 [{shapeType=Capsule, length=7.75, radius=2.25, shapeName=dyn_jacket_3_3_capsule, center=<5.0178 0 0>, rotation=<rot -0.0001 0 0>}]
 dyn_jacket_3_base [{shapeType=Capsule, length=0, radius=2.25, shapeName=dyn_jacket_3_base_capsule}]
 dyn_jacket_4_1 [{shapeType=Capsule, length=7.5, radius=2.5, shapeName=dyn_jacket_4_1_capsule, center=<4.0782 0 0>, rotation=<rot -0.0001 -0.0003 0>}]
 dyn_jacket_4_2 [{shapeType=Capsule, length=6.5, radius=3.5, shapeName=dyn_jacket_4_2_capsule, center=<4.0782 0 0>, rotation=<rot -0.0001 -0.0003 0>}]
 dyn_jacket_4_3 [{shapeType=Capsule, length=7.25, radius=2.75, shapeName=dyn_jacket_4_3_capsule, center=<5.0178 0 0>, rotation=<rot -0.0001 -0.0003 0>}]
 dyn_jacket_4_base [{shapeType=Capsule, length=0, radius=2.25, shapeName=dyn_jacket_4_base_capsule}]
 dyn_jacket_5_1 [{shapeType=Capsule, length=7, radius=3, shapeName=dyn_jacket_5_1_capsule, center=<4.0782 0 0>, rotation=<rot 0 -0.0003 0>}]
 dyn_jacket_5_2 [{shapeType=Capsule, length=6.5, radius=3.5, shapeName=dyn_jacket_5_2_capsule, center=<4.0782 0 0>, rotation=<rot 0 -0.0003 0>}]
 dyn_jacket_5_3 [{shapeType=Capsule, length=16.5, radius=2.75, shapeName=dyn_jacket_5_3_capsule, center=<5.0178 0 0>, rotation=<rot 0 -0.0003 0>}]
 dyn_jacket_5_base [{shapeType=Capsule, length=0, radius=2.25, shapeName=dyn_jacket_5_base_capsule}]
 dyn_jacket_6_1 [{shapeType=Capsule, length=7.5, radius=2.5, shapeName=dyn_jacket_6_1_capsule, center=<4.0782 0 0>, rotation=<rot 0.0001 -0.0003 0>}]
 dyn_jacket_6_2 [{shapeType=Capsule, length=6.5, radius=3.5, shapeName=dyn_jacket_6_2_capsule, center=<4.0782 0 0>, rotation=<rot 0.0001 -0.0003 0>}]
 dyn_jacket_6_3 [{shapeType=Capsule, length=7.25, radius=2.75, shapeName=dyn_jacket_6_3_capsule, center=<5.0178 0 0>, rotation=<rot 0.0001 -0.0003 0>}]
 dyn_jacket_6_base [{shapeType=Capsule, length=0, radius=2.25, shapeName=dyn_jacket_6_base_capsule}]
 dyn_jacket_7_1 [{shapeType=Capsule, length=7.75, radius=2.25, shapeName=dyn_jacket_7_1_capsule, center=<4.0782 0 0>, rotation=<rot 0.0001 0 0>}]
 dyn_jacket_7_2 [{shapeType=Capsule, length=7.5, radius=2.5, shapeName=dyn_jacket_7_2_capsule, center=<4.0782 0 0>, rotation=<rot 0.0001 0 0>}]
 dyn_jacket_7_3 [{shapeType=Capsule, length=7.75, radius=2.25, shapeName=dyn_jacket_7_3_capsule, center=<5.0178 0 0>, rotation=<rot 0.0001 0 0>}]
 dyn_jacket_7_base [{shapeType=Capsule, length=0, radius=2.25, shapeName=dyn_jacket_7_base_capsule}]
 dyn_jacket_8_1 [{shapeType=Capsule, length=7.75, radius=2.25, shapeName=dyn_jacket_8_1_capsule, center=<4.0782 0 0>, rotation=<rot 0.0001 0.0002 0>}]
 dyn_jacket_8_2 [{shapeType=Capsule, length=7.5, radius=2.5, shapeName=dyn_jacket_8_2_capsule, center=<4.0782 0 0>, rotation=<rot 0.0001 0.0002 0>}]
 dyn_jacket_8_3 [{shapeType=Capsule, length=6, radius=2.75, shapeName=dyn_jacket_8_3_capsule, center=<5.0178 0 0>, rotation=<rot 0.0003 0.0004 10>}]
 dyn_jacket_8_base [{shapeType=Capsule, length=0, radius=2.25, shapeName=dyn_jacket_8_base_capsule}]
 dyn_jacket_9_1 [{shapeType=Capsule, length=7.75, radius=2.25, shapeName=dyn_jacket_9_1_capsule, center=<4.0782 0 0>, rotation=<rot 0 0.0003 0>}]
 dyn_jacket_9_2 [{shapeType=Capsule, length=7.5, radius=2.5, shapeName=dyn_jacket_9_2_capsule, center=<4.0782 0 0>, rotation=<rot 0 0.0003 0>}]
 dyn_jacket_9_3 [{shapeType=Capsule, length=7.5, radius=2.75, shapeName=dyn_jacket_9_3_capsule, center=<5.1696 0 0>, rotation=<rot 0 0.0001 10>}]
 dyn_jacket_9_base [{shapeType=Capsule, length=0, radius=2.25, shapeName=dyn_jacket_9_base_capsule}]
 dyn_necklaces_1 [{shapeType=Capsule, length=6.25, radius=1.75, shapeName=dyn_necklaces_1_capsule, center=<2.7158 0 0>, rotation=<rot 0.0002 0 90>}]
 dyn_necklaces_2 [{shapeType=Capsule, length=6.5, radius=1.5, shapeName=dyn_necklaces_2_capsule, center=<2.7158 0 0>, rotation=<rot 0.0002 0 90>}]
 pelvis [{shapeType=Capsule, length=5, radius=12, shapeName=pelvis_capsule, center=<-3.6263 -1.2277 0>}]
 spine_01 [{shapeType=Capsule, length=8, radius=10.75, shapeName=spine_01_capsule, center=<0 0.2371 0.0642>}]
 spine_02 [{shapeType=Capsule, length=7.5, radius=10.25, shapeName=spine_02_capsule, center=<0 0.4332 0>}]
 spine_03 [{shapeType=Capsule, length=9.5, radius=9.25, shapeName=spine_03_capsule, center=<0 1.069 0>}]
 spine_05 [{shapeType=Capsule, length=20.75, radius=9.25, shapeName=spine_05_capsule}]
 thigh_l [{shapeType=Capsule, length=22.25, radius=7.75, shapeName=thigh_l_capsule, center=<-18.2897 -1.2292 0>, rotation=<rot -90 -90.3524 90.352>}]
 thigh_r [{shapeType=Capsule, length=28.5, radius=8.5, shapeName=thigh_r_capsule, center=<17.1939 0.8539 -0.8153>, rotation=<rot -85 -5.0002 0>}]
}
constraints [
 {bone1Name=dyn_jacket_1_1, bone2Name=dyn_jacket_1_base, swing1LimitDegrees=5, swing2LimitDegrees=5, twistLimitDegrees=2.5}
 {bone1Name=dyn_jacket_1_2, bone2Name=dyn_jacket_1_1, swing1LimitDegrees=1.5, swing2LimitDegrees=1.5, twistLimitDegrees=2.5}
 {
  bone1Name=dyn_jacket_1_3
  bone2Name=dyn_jacket_1_2
  swing1LimitDegrees=2.5
  swing1Motion=Locked
  swing2LimitDegrees=2.5
  swing2Motion=Locked
  twistLimitDegrees=2.5
  twistMotion=Locked
 }
 {bone1Name=dyn_jacket_2_1, bone2Name=dyn_jacket_2_base, swing1LimitDegrees=5, swing2LimitDegrees=5, twistLimitDegrees=2.5}
 {bone1Name=dyn_jacket_2_2, bone2Name=dyn_jacket_2_1, swing1LimitDegrees=1.5, swing2LimitDegrees=1.5, twistLimitDegrees=2.5}
 {bone1Name=dyn_jacket_2_3, bone2Name=dyn_jacket_2_2, swing1LimitDegrees=1.5, swing2LimitDegrees=1.5, twistLimitDegrees=2.5}
 {bone1Name=dyn_jacket_3_1, bone2Name=dyn_jacket_3_base, swing1LimitDegrees=5, swing2LimitDegrees=5, twistLimitDegrees=2.5}
 {bone1Name=dyn_jacket_3_2, bone2Name=dyn_jacket_3_1, swing1LimitDegrees=1.5, swing2LimitDegrees=1.5, twistLimitDegrees=2.5}
 {bone1Name=dyn_jacket_3_3, bone2Name=dyn_jacket_3_2, swing1LimitDegrees=0.5, swing2LimitDegrees=2.5, twistLimitDegrees=2.5}
 {bone1Name=dyn_jacket_4_2, bone2Name=dyn_jacket_4_1, swing1LimitDegrees=1.5, swing2LimitDegrees=1.5, twistLimitDegrees=1}
 {bone1Name=dyn_jacket_4_3, bone2Name=dyn_jacket_4_2, swing1LimitDegrees=2.5, swing2LimitDegrees=0.5, twistLimitDegrees=1}
 {bone1Name=dyn_jacket_5_1, bone2Name=dyn_jacket_5_base, swing1LimitDegrees=5, swing2LimitDegrees=5, twistLimitDegrees=2.5}
 {bone1Name=dyn_jacket_5_2, bone2Name=dyn_jacket_5_1, swing1LimitDegrees=1.5, swing2LimitDegrees=1.5, twistLimitDegrees=1}
 {bone1Name=dyn_jacket_5_3, bone2Name=dyn_jacket_5_2, swing1LimitDegrees=2.5, swing2LimitDegrees=0.5, twistLimitDegrees=1}
 {bone1Name=dyn_jacket_6_1, bone2Name=dyn_jacket_6_base, swing1LimitDegrees=5, swing2LimitDegrees=5, twistLimitDegrees=2.5}
 {bone1Name=dyn_jacket_6_2, bone2Name=dyn_jacket_6_1, swing1LimitDegrees=1.5, swing2LimitDegrees=1.5, twistLimitDegrees=1}
 {bone1Name=dyn_jacket_6_3, bone2Name=dyn_jacket_6_2, swing1LimitDegrees=2.5, swing2LimitDegrees=0.5, twistLimitDegrees=1}
 {bone1Name=dyn_jacket_7_1, bone2Name=dyn_jacket_7_base, swing1LimitDegrees=5, swing2LimitDegrees=5, twistLimitDegrees=2.5}
 {bone1Name=dyn_jacket_7_2, bone2Name=dyn_jacket_7_1, swing1LimitDegrees=1.5, swing2LimitDegrees=1.5, twistLimitDegrees=2.5}
 {bone1Name=dyn_jacket_7_3, bone2Name=dyn_jacket_7_2, swing1LimitDegrees=0.5, swing2LimitDegrees=2.5, twistLimitDegrees=2.5}
 {bone1Name=dyn_jacket_8_1, bone2Name=dyn_jacket_8_base, swing1LimitDegrees=5, swing2LimitDegrees=5, twistLimitDegrees=2.5}
 {bone1Name=dyn_jacket_8_2, bone2Name=dyn_jacket_8_1, swing1LimitDegrees=1.5, swing2LimitDegrees=1.5, twistLimitDegrees=2.5}
 {bone1Name=dyn_jacket_8_3, bone2Name=dyn_jacket_8_2, swing1LimitDegrees=1.5, swing2LimitDegrees=1.5, twistLimitDegrees=2.5}
 {bone1Name=dyn_jacket_9_1, bone2Name=dyn_jacket_9_base, swing1LimitDegrees=5, swing2LimitDegrees=5, twistLimitDegrees=10}
 {bone1Name=dyn_jacket_9_2, bone2Name=dyn_jacket_9_1, swing1LimitDegrees=1.5, swing2LimitDegrees=1.5, twistLimitDegrees=2.5}
 {bone1Name=dyn_jacket_4_1, bone2Name=dyn_jacket_4_base, swing1LimitDegrees=5, swing2LimitDegrees=5, twistLimitDegrees=2.5}
 {
  bone1Name=dyn_necklaces_1
  bone2Name=spine_05
  swing1LimitDegrees=45
  swing1Motion=Locked
  swing2LimitDegrees=45
  swing2Motion=Locked
  twistLimitDegrees=45
  twistMotion=Locked
 }
 {
  bone1Name=dyn_necklaces_2
  bone2Name=dyn_necklaces_1
  swing1LimitDegrees=45
  swing1Motion=Locked
  swing2LimitDegrees=45
  swing2Motion=Locked
  twistLimitDegrees=45
  twistMotion=Locked
 }
 {bone1Name=dyn_chain_control_1, bone2Name=dyn_chain_control_2, swing1LimitDegrees=45, swing2LimitDegrees=45, twistLimitDegrees=45}
 {bone1Name=dyn_chain_control_2, bone2Name=dyn_chain_control_3, swing1LimitDegrees=30, swing2LimitDegrees=30, twistLimitDegrees=20}
 {
  bone1Name=dyn_jacket_9_3
  bone2Name=dyn_jacket_9_2
  swing1LimitDegrees=2.5
  swing1Motion=Locked
  swing2LimitDegrees=2.5
  swing2Motion=Locked
  twistLimitDegrees=2.5
 }
]
