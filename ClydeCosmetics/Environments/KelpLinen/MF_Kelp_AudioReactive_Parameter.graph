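# Material function for Kelp Linen environment materials that outputs looping, music-timed beat values for pulsing effects.
# Main_Phase_Tracker selects a time-based tempo (Phase3 differs), whose fraction samples an ADSR curve atlas row.
# AudioBeat adds a rerouted offset blend and picks an RGB channel via BeatPicker; Beat_Tracker outputs plain red.
# EventAudio_Spectrum remaps, per-instance randomness and emissive ranges are also built, but never reach either output.
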MaterialFunction /ClydeCosmetics/Environments/KelpLinen/MF_Kelp_AudioReactive_Parameter
graph {nodes [
  n1 = CollectionParameter() {collection=&FortniteMaterialParameters, parameterName=Main_Phase_Tracker}
  n2 = Constant() {r=1.4286}
  n3 = Divide(A=Time(), B=@n2)
  n4 = Divide(A=@n2)
  n5 = NamedRerouteUsage()
  MaterialExpressionAdd_6 = Add(A=LinearInterpolate(A=@n3, B=Add(A=@n3, B=@n4), Alpha=@n5), B=Multiply(A=@n2) {constB=-0.1})
  n6 = Constant() {r=1.3043}
  n7 = Divide(A=Time(), B=@n6)
  n8 = Divide(A=@n6)
  MaterialExpressionAdd_7 = Add(A=LinearInterpolate(A=@n7, B=Add(A=@n7, B=@n8), Alpha=@n5), B=Multiply(A=@n6) {constB=-0.1})
  n9 = Switch!(
   SwitchValue=@n1
   Default
   Empty=@MaterialExpressionAdd_6
   Phase1=@MaterialExpressionAdd_6
   Phase2=@MaterialExpressionAdd_6
   Phase3=@MaterialExpressionAdd_7
  ) {inputs [
    {
     inputName=Empty
     input {
      expression=&self:MaterialExpressionAdd_6
      inputName=None
      mask=0
      maskA=0
      maskB=0
      maskG=0
      maskR=0
      outputIndex=0
     }
    }
    {
     inputName=Phase1
     input {
      expression=&self:MaterialExpressionAdd_6
      inputName=None
      mask=0
      maskA=0
      maskB=0
      maskG=0
      maskR=0
      outputIndex=0
     }
    }
    {
     inputName=Phase2
     input {
      expression=&self:MaterialExpressionAdd_6
      inputName=None
      mask=0
      maskA=0
      maskB=0
      maskG=0
      maskR=0
      outputIndex=0
     }
    }
    {
     inputName=Phase3
     input {
      expression=&self:MaterialExpressionAdd_7
      inputName=None
      mask=0
      maskA=0
      maskB=0
      maskG=0
      maskR=0
      outputIndex=0
     }
    }
   ]}
  MaterialExpressionCurveAtlasRowParameter_2 = CurveAtlasRowParameter(CurveTime=Frac(None=@n9)) {
   atlas=&/CurveAtlas_KelpLinen
   curve=&/Curve_KelpLinen_ADSR_AudioReactive
  }
  n10 = Switch!(
   SwitchValue=ScalarParameter() {parameterName=BeatPicker}
   Default
   Red=@MaterialExpressionCurveAtlasRowParameter_2.R
   Green=@MaterialExpressionCurveAtlasRowParameter_2.R
   Blue=@MaterialExpressionCurveAtlasRowParameter_2.R
  ) {inputs [
    {
     inputName=Red
     input {
      expression=&self:MaterialExpressionCurveAtlasRowParameter_2
      inputName=None
      mask=1
      maskA=0
      maskB=0
      maskG=0
      maskR=1
      outputIndex=1
     }
    }
    {
     inputName=Green
     input {
      expression=&self:MaterialExpressionCurveAtlasRowParameter_2
      inputName=None
      mask=1
      maskA=0
      maskB=0
      maskG=1
      maskR=0
      outputIndex=2
     }
    }
    {
     inputName=Blue
     input {
      expression=&self:MaterialExpressionCurveAtlasRowParameter_2
      inputName=None
      mask=1
      maskA=0
      maskB=1
      maskG=0
      maskR=0
      outputIndex=3
     }
    }
   ]}
  FunctionOutput(None=NamedRerouteDeclaration(None=@n10) {name=AudioBeat, nodeColor=<rgba 1 0.8438 0 1>})
  n11 = CollectionParameter() {collection=&FortniteMaterialParameters, parameterName=Main_Phase_Tracker}
  n12 = Constant() {r=1.4286}
  MaterialExpressionAdd_8 = Add(A=Divide(A=Time(), B=@n12), B=Multiply(A=@n12) {constB=-0.1})
  n13 = Constant() {r=1.3043}
  MaterialExpressionAdd_5 = Add(A=Divide(A=Time(), B=@n13), B=Multiply(A=@n13) {constB=-0.1})
  n14 = Switch!(
   SwitchValue=@n11
   Default
   Empty=@MaterialExpressionAdd_8
   Phase1=@MaterialExpressionAdd_8
   Phase2=@MaterialExpressionAdd_8
   Phase3=@MaterialExpressionAdd_5
  ) {inputs [
    {
     inputName=Empty
     input {
      expression=&self:MaterialExpressionAdd_8
      inputName=None
      mask=0
      maskA=0
      maskB=0
      maskG=0
      maskR=0
      outputIndex=0
     }
    }
    {
     inputName=Phase1
     input {
      expression=&self:MaterialExpressionAdd_8
      inputName=None
      mask=0
      maskA=0
      maskB=0
      maskG=0
      maskR=0
      outputIndex=0
     }
    }
    {
     inputName=Phase2
     input {
      expression=&self:MaterialExpressionAdd_8
      inputName=None
      mask=0
      maskA=0
      maskB=0
      maskG=0
      maskR=0
      outputIndex=0
     }
    }
    {
     inputName=Phase3
     input {
      expression=&self:MaterialExpressionAdd_5
      inputName=None
      mask=0
      maskA=0
      maskB=0
      maskG=0
      maskR=0
      outputIndex=0
     }
    }
   ]}
  n15 = CurveAtlasRowParameter(CurveTime=Frac(None=@n14)) {
   atlas=&CurveAtlas_KelpLinen
   curve=&Curve_KelpLinen_ADSR_AudioReactive
  }
  FunctionOutput(None=@n15.R) {description=Used for elements that want to follow the beat, outputName=Beat_Tracker}
  n16 = CollectionParameter() {collection=&FortniteMaterialParameters, parameterName=EventAudio_Spectrum}
  n17 = Multiply(A=@n16)
  n18 = MaterialFunctionCall(
   Input=ComponentMask(None=@n17) {b=true}
   Input Low=Constant() {r=0.1}
   Input High=Constant() {r=0.25}
   Target Low=Constant()
   Target High=Constant() {r=1}
  ) {materialFunction=&RemapValueRange}
  LinearInterpolate(
   A=@n18.Result
   B=MaterialFunctionCall(
    Input=ComponentMask(None=@n17) {b=true}
    Input Low=Constant() {r=0.2}
    Input High=Constant() {r=0.5}
    Target Low=Constant()
    Target High=Constant() {r=1}
   ) {materialFunction=&RemapValueRange}.Result
   Alpha=ScalarParameter() {defaultValue=1, parameterName=Phase_Music}
  )
  n19 = ComponentMask(None=@n17) {r=true}
  n20 = ScalarParameter() {defaultValue=0.7, parameterName=LowFQ_Input_Low}
  n21 = ScalarParameter() {defaultValue=0.82, parameterName=LowFQ_Input_High}
  n22 = Constant()
  n23 = Constant() {r=1}
  n24 = ComponentMask(None=@n17) {b=true}
  n25 = ScalarParameter() {defaultValue=0.35, parameterName=HighFQ_Input_Low}
  n26 = ScalarParameter() {defaultValue=0.5, parameterName=HighFQ_Input_High}
  n27 = PerInstanceRandom()
  n28 = CollectionParameter() {collection=&FortniteMaterialParameters, parameterName=Main_Phase_Tracker}
  n29 = Time()
  MaterialExpressionMultiply_1 = Multiply(A=Constant() {r=2.8571}, B=@n29)
  MaterialExpressionMultiply_4 = Multiply(A=@n29) {constB=2.6087}
  n30 = Switch(
   SwitchValue=@n28
   Empty=@MaterialExpressionMultiply_1
   Phase1=@MaterialExpressionMultiply_1
   Phase2=@MaterialExpressionMultiply_1
   Phase3=@MaterialExpressionMultiply_4
  ) {inputs [
    {
     inputName=Empty
     input {
      expression=&self:MaterialExpressionMultiply_1
      inputName=None
      mask=0
      maskA=0
      maskB=0
      maskG=0
      maskR=0
      outputIndex=0
     }
    }
    {
     inputName=Phase1
     input {
      expression=&self:MaterialExpressionMultiply_1
      inputName=None
      mask=0
      maskA=0
      maskB=0
      maskG=0
      maskR=0
      outputIndex=0
     }
    }
    {
     inputName=Phase2
     input {
      expression=&self:MaterialExpressionMultiply_1
      inputName=None
      mask=0
      maskA=0
      maskB=0
      maskG=0
      maskR=0
      outputIndex=0
     }
    }
    {
     inputName=Phase3
     input {
      expression=&self:MaterialExpressionMultiply_4
      inputName=None
      mask=0
      maskA=0
      maskB=0
      maskG=0
      maskR=0
      outputIndex=0
     }
    }
   ]}
  n31 = CurveAtlasRowParameter(CurveTime=Frac(None=@n30)) {
   atlas=&CurveAtlas_KelpLinen
   curve=&Curve_KelpLinen_ADSR_AudioReactive
  }
  n32 = LinearInterpolate(
   B=LinearInterpolate(
    A=MaterialFunctionCall(Input=@n19, Input Low=@n20, Input High=@n21, Target Low=@n22, Target High=@n23) {materialFunction=&RemapValueRange}.Result
    B=MaterialFunctionCall(Input=@n24, Input Low=@n25, Input High=@n26, Target Low=@n22, Target High=@n23) {materialFunction=&RemapValueRange}.Result
    Alpha=Round(None=@n27)
   )
   Alpha=@n31
  )
  LinearInterpolate(
   A=ScalarParameter() {defaultValue=1, parameterName=Reactive_Emissive_Low}
   B=ScalarParameter() {defaultValue=3, parameterName=Reactive_Emissive_High}
   Alpha=@n32
  ) {constA=1, constB=3}
  n33 = CollectionParameter() {collection=&FortniteMaterialParameters, parameterName=Main_Phase_Tracker}
  n34 = Time()
  MaterialExpressionMultiply_7 = Multiply(A=@n34) {constB=0.7143}
  MaterialExpressionMultiply_6 = Multiply(A=@n34) {constB=0.6522}
  n35 = Switch(
   SwitchValue=@n33
   Empty=@MaterialExpressionMultiply_7
   Phase1=@MaterialExpressionMultiply_7
   Phase2=@MaterialExpressionMultiply_7
   Phase3=@MaterialExpressionMultiply_6
  ) {inputs [
    {
     inputName=Empty
     input {
      expression=&self:MaterialExpressionMultiply_7
      inputName=None
      mask=0
      maskA=0
      maskB=0
      maskG=0
      maskR=0
      outputIndex=0
     }
    }
    {
     inputName=Phase1
     input {
      expression=&self:MaterialExpressionMultiply_7
      inputName=None
      mask=0
      maskA=0
      maskB=0
      maskG=0
      maskR=0
      outputIndex=0
     }
    }
    {
     inputName=Phase2
     input {
      expression=&self:MaterialExpressionMultiply_7
      inputName=None
      mask=0
      maskA=0
      maskB=0
      maskG=0
      maskR=0
      outputIndex=0
     }
    }
    {
     inputName=Phase3
     input {
      expression=&self:MaterialExpressionMultiply_6
      inputName=None
      mask=0
      maskA=0
      maskB=0
      maskG=0
      maskR=0
      outputIndex=0
     }
    }
   ]}
  n36 = CurveAtlasRowParameter(CurveTime=Frac(None=@n35)) {
   atlas=&CurveAtlas_KelpLinen
   curve=&Curve_KelpLinen_ADSR_AudioReactive
  }
  LinearInterpolate(B=Clamp(None=@n18.Result), Alpha=@n36)
  Multiply(
   A=@n32
   B=LinearInterpolate(
    A=ScalarParameter() {defaultValue=0.5, group=Kelp, parameterName=PerInstance_MinAmount, sortPriority=34}
    B=ScalarParameter() {defaultValue=1, group=Kelp, parameterName=PerInstance_MaxAmount, sortPriority=35}
    Alpha=PerInstanceRandom()
   ) {constA=0.5}
  )
  n37 = Frac(None=MaterialFunctionCall(f3=ObjectPositionWS()) {materialFunction=&AddComponents}.f3)
  n38 = PerInstanceRandom()
  NamedRerouteDeclaration(None=Round(None=StaticSwitchParameter(True=@n37, False=@n38) {parameterName=Prop})) {name=RandomOffset, nodeColor=<rgba 0 1 0.9766 1>}
  ScalarParameter() {parameterName=Alternative_BeatFrequency}
 ]}
